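# Physics asset for the belt mesh of the female medium ChemPencil cosmetic body, covering belt straps, a bottle and left-side skirt panels.
# Capsules on pelvis, thighs and dynamic bones form chains rooted at the pelvis, most through base or offset bones.
# Chain roots and the side-A skirt joints swing freely, while most later segments have tight swing limits and zero twist.
# The bottle hangs off the second segment of the right back belt.
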PhysicsAsset /BRCosmetics/Characters/Player/Female/Medium/Bodies/F_MED_ChemPencil/Meshes/F_MED_ChemPencil_Belt_Physics
bodies {
 base_belt_back_lf [{shapeType=Capsule, length=0.505, radius=0.505, shapeName=base_belt_back_lf_capsule, rotation=<rot 0 -90 0>}]
 base_belt_back_rt [{shapeType=Capsule, length=0.505, radius=0.505, shapeName=base_belt_back_rt_capsule, rotation=<rot 0 -90 0>}]
 base_belt_front_lf [{shapeType=Capsule, length=0.505, radius=0.505, shapeName=base_belt_front_lf_capsule, rotation=<rot 0 -90 0>}]
 dyn_belt_back_01_lf [{
   shapeType=Capsule
   length=1.6859
   radius=1.5666
   shapeName=dyn_belt_back_01_lf_capsule
   center=<2.5921 0.0089 0.0543>
   rotation=<rot 0 177.979 -85.5806>
  }]
 dyn_belt_back_01_rt [{
   shapeType=Capsule
   length=1.7528
   radius=1.4882
   shapeName=dyn_belt_back_01_rt_capsule
   center=<3.1072 0.0104 0.1672>
   rotation=<rot 0 1.2879 -88.3506>
  }]
 dyn_belt_back_02_lf [{shapeType=Capsule, length=5, radius=2, shapeName=dyn_belt_back_02_lf_capsule, center=<2.1025 0 1.0298>, rotation=<rot 90 0 0>}]
 dyn_belt_back_02_rt [{shapeType=Capsule, length=5.2743, radius=1.7257, shapeName=dyn_belt_back_02_rt_capsule, center=<4.5079 0 -0.5415>, rotation=<rot 90 0 0>}]
 dyn_belt_back_03_lf [{shapeType=Capsule, length=5, radius=2, shapeName=dyn_belt_back_03_lf_capsule, center=<5 0 0>, rotation=<rot 90 0 0>}]
 dyn_belt_back_03_rt [{shapeType=Capsule, length=5.3005, radius=1.6995, shapeName=dyn_belt_back_03_rt_capsule, center=<5 0 -0.5853>, rotation=<rot 90 0 0>}]
 dyn_belt_back_04_lf [{shapeType=Capsule, length=5, radius=2, shapeName=dyn_belt_back_04_lf_capsule, center=<5 0 0>, rotation=<rot 90 0 0>}]
 dyn_belt_back_04_rt [{
   shapeType=Capsule
   length=4.7063
   radius=2.2937
   shapeName=dyn_belt_back_04_rt_capsule
   center=<5 0 -0.8357>
   rotation=<rot 90 179.9991 179.9991>
  }]
 dyn_belt_back_05_lf [{shapeType=Capsule, length=5, radius=2, shapeName=dyn_belt_back_05_lf_capsule, center=<5 0 0>, rotation=<rot 90 0 0>}]
 dyn_belt_back_05_rt [{shapeType=Capsule, length=4.8825, radius=3.044, shapeName=dyn_belt_back_05_rt_capsule, center=<5 0 -1.0307>, rotation=<rot 90 0 0>}]
 dyn_belt_front_01_lf [{shapeType=Capsule, length=0.505, radius=0.505, shapeName=dyn_belt_front_01_lf_capsule, rotation=<rot 0 -90 0>}]
 dyn_belt_front_02_lf [{shapeType=Capsule, length=5, radius=1, shapeName=dyn_belt_front_02_lf_capsule, center=<5 0 0>, rotation=<rot 90 0 0>}]
 dyn_belt_front_03_lf [{shapeType=Capsule, length=5, radius=1, shapeName=dyn_belt_front_03_lf_capsule, center=<6.5087 0 0>, rotation=<rot 90 0 0>}]
 dyn_belt_front_04_lf [{shapeType=Capsule, length=5, radius=1, shapeName=dyn_belt_front_04_lf_capsule, center=<5 0 0>, rotation=<rot 90 0 0>}]
 dyn_belt_front_05_lf [{
   shapeType=Capsule
   length=4.8502
   radius=1.1498
   shapeName=dyn_belt_front_05_lf_capsule
   center=<2.134 0 0>
   rotation=<rot 90 -165.9638 -75.9639>
  }]
 dyn_bottle_rt [{
   shapeType=Capsule
   length=9.7996
   radius=1.7686
   shapeName=dyn_bottle_rt_capsule
   center=<8.8132 0.4939 1.508>
   rotation=<rot 0 90.5167 -100.7605>
  }]
 dyn_skirt_front_01_01_lf [{
   shapeType=Capsule
   length=3.1421
   radius=2.5319
   shapeName=dyn_skirt_front_01_01_lf_capsule
   center=<6.5027 -2.5244 0.543>
   rotation=<rot 0 -157.7888 -89.9313>
  }]
 dyn_skirt_front_02_lf [{shapeType=Capsule, length=2.9251, radius=4.1479, shapeName=dyn_skirt_front_02_lf_capsule, center=<4.4173 0 0>, rotation=<rot 90 180 90>}]
 dyn_skirt_front_back_01_lf [{
   shapeType=Capsule
   length=4.7215
   radius=8.3463
   shapeName=dyn_skirt_front_back_01_lf_capsule
   center=<13.3939 -2.4617 -4.1575>
   rotation=<rot -2.5945 -82.0168 -95.2635>
  }]
 dyn_skirt_front_back_02_lf [{
   shapeType=Capsule
   length=11.7614
   radius=7.4619
   shapeName=dyn_skirt_front_back_02_lf_capsule
   center=<3.9626 -0.0408 -6.2558>
   rotation=<rot 80 84.9994 -0.0009>
  }]
 dyn_skirt_front_side_01_01_lf [{
   shapeType=Capsule
   length=9.9798
   radius=7.1354
   shapeName=dyn_skirt_front_side_01_01_lf_capsule
   center=<-1.5914 0 -7.8152>
   rotation=<rot 0 -89.9998 -90.0002>
  }]
 dyn_skirt_front_side_01_lf [{
   shapeType=Capsule
   length=9.9232
   radius=5.8882
   shapeName=dyn_skirt_front_side_01_lf_capsule
   center=<15.387 1.0136 -4.1608>
   rotation=<rot -31.9793 -90.0943 -100.4264>
  }]
 dyn_skirt_front_side_02_lf [{
   shapeType=Capsule
   length=12.6897
   radius=7.5744
   shapeName=dyn_skirt_front_side_02_lf_capsule
   center=<2.6339 -0.099 -7.0242>
   rotation=<rot 84.9999 -90.0026 -0.0031>
  }]
 dyn_skirt_side_A_01_lf [{
   shapeType=Capsule
   length=8
   radius=5
   shapeName=dyn_skirt_side_A_01_lf_capsule
   center=<9.7071 1.6888 -4.9017>
   rotation=<rot 67.7313 152.7271 149.4944>
  }]
 dyn_skirt_side_A_02_lf [{
   shapeType=Capsule
   length=9.5178
   radius=5.2161
   shapeName=dyn_skirt_side_A_02_lf_capsule
   center=<0.8681 -2.2768 -7.0762>
   rotation=<rot 85 -4.9999 -5.0002>
  }]
 offset_skirt_front_back_lf [{shapeType=Capsule, length=0.505, radius=0.505, shapeName=offset_skirt_front_back_lf_capsule, rotation=<rot 0 -90 0>}]
 offset_skirt_front_lf [{shapeType=Capsule, length=0.505, radius=0.505, shapeName=offset_skirt_front_lf_capsule, rotation=<rot 0 -90 0>}]
 offset_skirt_front_side_lf [{shapeType=Capsule, length=0.505, radius=0.505, shapeName=offset_skirt_front_side_lf_capsule, rotation=<rot 0 -90 0>}]
 pelvis [{shapeType=Capsule, length=11.9387, radius=8.5108, center=<-4.8052 -1.9507 0>, rotation=<rot 0 -90 0>}]
 thigh_l [{shapeType=Capsule, length=20.1186, radius=8.6724, shapeName=thigh_l_capsule, center=<-22.6756 -1.4192 0>, rotation=<rot 0 -92.5819 -95>}]
 thigh_r [{shapeType=Capsule, length=20.9269, radius=7.875, center=<11.3635 2.8903 -1.8135>, rotation=<rot -0.4308 -88.4701 85.0693>}]
}
constraints [
 {bone1Name=thigh_r, bone2Name=pelvis, swing1LimitDegrees=45, swing2LimitDegrees=45, twistLimitDegrees=45}
 {
  bone1Name=offset_skirt_front_back_lf
  bone2Name=pelvis
  swing1LimitDegrees=45
  swing1Motion=Free
  swing2LimitDegrees=45
  swing2Motion=Free
  twistLimitDegrees=45
  twistMotion=Free
 }
 {
  bone1Name=dyn_skirt_front_back_01_lf
  bone2Name=offset_skirt_front_back_lf
  swing1LimitDegrees=45
  swing1Motion=Free
  swing2LimitDegrees=45
  swing2Motion=Free
  twistLimitDegrees=45
  twistMotion=Free
 }
 {bone1Name=dyn_skirt_front_back_02_lf, bone2Name=dyn_skirt_front_back_01_lf, swing1LimitDegrees=0, swing2LimitDegrees=20, twistLimitDegrees=0}
 {
  bone1Name=offset_skirt_front_lf
  bone2Name=pelvis
  swing1LimitDegrees=45
  swing1Motion=Free
  swing2LimitDegrees=45
  swing2Motion=Free
  twistLimitDegrees=45
  twistMotion=Free
 }
 {bone1Name=dyn_skirt_front_02_lf, bone2Name=dyn_skirt_front_01_01_lf, swing1LimitDegrees=0, swing2LimitDegrees=20, twistLimitDegrees=0}
 {
  bone1Name=offset_skirt_front_side_lf
  bone2Name=pelvis
  swing1LimitDegrees=45
  swing1Motion=Free
  swing2LimitDegrees=45
  swing2Motion=Free
  twistLimitDegrees=45
  twistMotion=Free
 }
 {
  bone1Name=dyn_skirt_front_side_01_lf
  bone2Name=offset_skirt_front_side_lf
  swing1LimitDegrees=45
  swing1Motion=Free
  swing2LimitDegrees=45
  swing2Motion=Free
  twistLimitDegrees=45
  twistMotion=Free
 }
 {
  bone1Name=dyn_skirt_front_side_02_lf
  bone2Name=dyn_skirt_front_side_01_01_lf
  swing1LimitDegrees=0
  swing2LimitDegrees=20
  twistLimitDegrees=0
 }
 {
  bone1Name=base_belt_back_lf
  bone2Name=pelvis
  swing1LimitDegrees=45
  swing1Motion=Free
  swing2LimitDegrees=45
  swing2Motion=Free
  twistLimitDegrees=45
  twistMotion=Free
 }
 {
  bone1Name=dyn_belt_back_01_lf
  bone2Name=base_belt_back_lf
  swing1LimitDegrees=45
  swing1Motion=Free
  swing2LimitDegrees=45
  swing2Motion=Free
  twistLimitDegrees=45
  twistMotion=Free
 }
 {bone1Name=dyn_belt_back_02_lf, bone2Name=dyn_belt_back_01_lf, swing1LimitDegrees=0, swing2LimitDegrees=5, twistLimitDegrees=0}
 {bone1Name=dyn_belt_back_03_lf, bone2Name=dyn_belt_back_02_lf, swing1LimitDegrees=0, swing2LimitDegrees=10, twistLimitDegrees=0}
 {bone1Name=dyn_belt_back_04_lf, bone2Name=dyn_belt_back_03_lf, swing1LimitDegrees=0, swing2LimitDegrees=20, twistLimitDegrees=0}
 {bone1Name=dyn_belt_back_05_lf, bone2Name=dyn_belt_back_04_lf, swing1LimitDegrees=0, swing2LimitDegrees=15, twistLimitDegrees=0}
 {
  bone1Name=base_belt_back_rt
  bone2Name=pelvis
  swing1LimitDegrees=45
  swing1Motion=Free
  swing2LimitDegrees=45
  swing2Motion=Free
  twistLimitDegrees=45
  twistMotion=Free
 }
 {
  bone1Name=dyn_belt_back_01_rt
  bone2Name=base_belt_back_rt
  swing1LimitDegrees=45
  swing1Motion=Free
  swing2LimitDegrees=45
  swing2Motion=Free
  twistLimitDegrees=45
  twistMotion=Free
 }
 {bone1Name=dyn_belt_back_02_rt, bone2Name=dyn_belt_back_01_rt, swing1LimitDegrees=10, swing2LimitDegrees=10, twistLimitDegrees=0}
 {bone1Name=dyn_belt_back_03_rt, bone2Name=dyn_belt_back_02_rt, swing1LimitDegrees=10, swing2LimitDegrees=15, twistLimitDegrees=0}
 {bone1Name=dyn_belt_back_04_rt, bone2Name=dyn_belt_back_03_rt, swing1LimitDegrees=20, swing2LimitDegrees=25, twistLimitDegrees=0}
 {bone1Name=dyn_belt_back_05_rt, bone2Name=dyn_belt_back_04_rt, swing1LimitDegrees=0, swing2LimitDegrees=25, twistLimitDegrees=0}
 {
  bone1Name=dyn_skirt_front_01_01_lf
  bone2Name=offset_skirt_front_lf
  swing1LimitDegrees=45
  swing1Motion=Free
  swing2LimitDegrees=45
  swing2Motion=Free
  twistLimitDegrees=45
  twistMotion=Free
 }
 {
  bone1Name=dyn_skirt_front_side_01_01_lf
  bone2Name=dyn_skirt_front_side_01_lf
  swing1LimitDegrees=45
  swing1Motion=Free
  swing2LimitDegrees=45
  swing2Motion=Free
  twistLimitDegrees=45
  twistMotion=Free
 }
 {bone1Name=dyn_bottle_rt, bone2Name=dyn_belt_back_02_rt, swing1LimitDegrees=0, swing2LimitDegrees=20, twistLimitDegrees=0}
 {
  bone1Name=base_belt_front_lf
  bone2Name=pelvis
  swing1LimitDegrees=45
  swing1Motion=Free
  swing2LimitDegrees=45
  swing2Motion=Free
  twistLimitDegrees=45
  twistMotion=Free
 }
 {
  bone1Name=dyn_belt_front_01_lf
  bone2Name=base_belt_front_lf
  swing1LimitDegrees=45
  swing1Motion=Free
  swing2LimitDegrees=45
  swing2Motion=Free
  twistLimitDegrees=45
  twistMotion=Free
 }
 {bone1Name=dyn_belt_front_02_lf, bone2Name=dyn_belt_front_01_lf, swing1LimitDegrees=0, swing2LimitDegrees=15, twistLimitDegrees=0}
 {bone1Name=dyn_belt_front_03_lf, bone2Name=dyn_belt_front_02_lf, swing1LimitDegrees=0, swing2LimitDegrees=15, twistLimitDegrees=0}
 {bone1Name=dyn_belt_front_04_lf, bone2Name=dyn_belt_front_03_lf, swing1LimitDegrees=0, swing2LimitDegrees=15, twistLimitDegrees=0}
 {bone1Name=dyn_belt_front_05_lf, bone2Name=dyn_belt_front_04_lf, swing1LimitDegrees=0, swing2LimitDegrees=15, twistLimitDegrees=0}
 {
  bone1Name=thigh_l
  bone2Name=pelvis
  swing1LimitDegrees=45
  swing1Motion=Free
  swing2LimitDegrees=45
  swing2Motion=Free
  twistLimitDegrees=45
  twistMotion=Free
 }
 {
  bone1Name=dyn_skirt_side_A_01_lf
  bone2Name=pelvis
  swing1LimitDegrees=45
  swing1Motion=Free
  swing2LimitDegrees=45
  swing2Motion=Free
  twistLimitDegrees=45
  twistMotion=Free
 }
 {
  bone1Name=dyn_skirt_side_A_02_lf
  bone2Name=dyn_skirt_side_A_01_lf
  swing1LimitDegrees=45
  swing1Motion=Free
  swing2LimitDegrees=45
  swing2Motion=Free
  twistLimitDegrees=45
  twistMotion=Free
 }
]
