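# Control Rig for the F_MED_ChemPencil female character body, adding procedural correctives for the skirt, belt and neck bones on top of a deform rig.
# After an optional debug-control pass, it runs the shared DeformRig function, parent-constrains skirt base bones to spine and pelvis, then solves skirt and neck.
# Spherical pose readers and remapped joint rotations weight additive transforms and rotation constraints, with emote, frontend, Crouch and RBAN variables adjusting the left-side skirt response.
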ControlRigBlueprint /BRCosmetics/Characters/Player/Female/Medium/Bodies/F_MED_ChemPencil/Meshes/F_MED_ChemPencil_CtrlRig
graphs [
 New Function_ContainedGraph {nodes [
   n1 = RigVMFunctionEntryNode()
   RigVMFunction_ControlFlowBranch = &RigVMFunction_ControlFlowBranch(Condition=@n1.Debug, ExecuteContext=@n1.ExecuteContext)
   RigVMFunctionReturnNode(ExecuteContext=@RigVMFunction_ControlFlowBranch.Completed)
   RigVMFunction_Sequence = &RigVMFunction_Sequence(ExecuteContext=@RigVMFunction_ControlFlowBranch.True)
   RigUnit_GetTransform = &RigUnit_GetTransform(Item=(Type=Control,Name="thigh_l_ctrl"), bInitial=False)
   RigUnit_SetTransform = &RigUnit_SetTransform(Item=(Type=Bone,Name="thigh_l"), Weight=1, bInitial=False, Value=@RigUnit_GetTransform.Transform, ExecutePin=@RigVMFunction_Sequence.A)
   > &RigUnit_SetTransform(
    Item=(Type=Bone,Name="thigh_r")
    Weight=1
    bInitial=False
    Value=&RigUnit_GetTransform(Item=(Type=Control,Name="thigh_r_ctrl"), bInitial=False).Transform
   )
   > &RigUnit_SetTransform(
    Item=(Type=Bone,Name="calf_r")
    Weight=1
    bInitial=False
    Value=&RigUnit_GetTransform(Item=(Type=Control,Name="calf_r_ctrl"), bInitial=False).Transform
   )
   > &RigUnit_SetTransform(
    Item=(Type=Bone,Name="calf_l")
    Weight=1
    bInitial=False
    Value=&RigUnit_GetTransform(Item=(Type=Control,Name="calf_l_ctrl"), bInitial=False).Transform
   )
   &RigUnit_SetTransform(
    Item=(Type=Bone,Name="clavicle_r")
    Weight=1
    bInitial=False
    Value=&RigUnit_GetTransform(Item=(Type=Control,Name="clavicle_r_ctrl"), bInitial=False).Transform
    ExecutePin=@RigVMFunction_Sequence.B
   )
   > &RigUnit_SetTransform(
    Item=(Type=Bone,Name="clavicle_l")
    Weight=1
    bInitial=False
    Value=&RigUnit_GetTransform(Item=(Type=Control,Name="clavicle_l_ctrl"), bInitial=False).Transform
   )
   > &RigUnit_SetTransform(
    Item=(Type=Bone,Name="neck_01")
    Weight=1
    bInitial=False
    Value=&RigUnit_GetTransform(Item=(Type=Control,Name="neck_01_ctrl"), bInitial=False).Transform
   )
   > &RigUnit_SetTransform(
    Item=(Type=Bone,Name="neck_02")
    Weight=1
    bInitial=False
    Value=&RigUnit_GetTransform(Item=(Type=Control,Name="neck_02_ctrl"), bInitial=False).Transform
   )
   > &RigUnit_SetTransform(
    Item=(Type=Bone,Name="head")
    Weight=1
    bInitial=False
    Value=&RigUnit_GetTransform(Item=(Type=Control,Name="head_ctrl"), bInitial=False).Transform
   )
  ]}
 New Function_ContainedGraph {nodes [
   n1 = RigVMFunctionEntryNode()
   RigVMFunction_Sequence = &RigVMFunction_Sequence(ExecuteContext=@n1.ExecuteContext)
   RigVMFunction_Sequence_4 = RigVMAggregateNode(ExecuteContext=@RigVMFunction_Sequence.A)
   n2 = &RigUnit_SphericalPoseReader(
    ActiveRegionScaleFactors=(PositiveWidth=1,NegativeWidth=0.7,PositiveHeight=1.5,NegativeHeight=1.5)
    ActiveRegionSize=0.2
    Debug=(bDrawDebug=true,bDraw2D=false,bDrawLocalAxes=False,DebugScale=25,DebugSegments=20,DebugThickness=0.25)
    DriverAxis=(X=-1,Y=0,Z=0)
    DriverItem=(Type=Bone,Name="thigh_l")
    FalloffRegionScaleFactors=(PositiveWidth=0,NegativeWidth=1,PositiveHeight=1,NegativeHeight=0.7)
    FalloffSize=0.67
    FlipHeightScaling=false
    FlipWidthScaling=false
    OptionalParentItem=(Type=Bone,Name="None")
    RotationOffset=(X=-120,Y=0,Z=30)
    ExecutePin=@RigVMFunction_Sequence_4.A
   )
   n3 = &RigVMFunction_AnimEvalRichCurve(
    Curve="(EditorCurveData=(Keys=((InterpMode=RCIM_Cubic,TangentMode=RCTM_Break,TangentWeightMode=RCTWM_WeightedNone,Time=0,Value=0,ArriveTangent=0,ArriveTangentWeight=0,LeaveTangent=0,LeaveTangentWeight=0),(InterpMode=RCIM_Cubic,TangentMode=RCTM_Break,TangentWeight ...(truncated)"
    SourceMaximum=0.05
    SourceMinimum=0
    TargetMinimum=0.7
    Value=RigVMDispatchNode(BoneName=root, Default=0, Name=deformRig_l_thigh_up_alpha) {template="DISPATCH_RigDispatch_GetAnimAttribute(in Name,in BoneName,in Default,out Value,out Found)"}.Value
   )
   n4 = RigVMDispatchNode(False=1, Condition=RigVMVariableNode() {variable=emote}.Value, True=@n3.Result) {template="DISPATCH_RigVMDispatch_If(in Condition,in True,in False,out Result)"}
   > &RigUnit_RotationConstraintLocalSpaceOffset(
    AdvancedSettings=(InterpolationType=Average,RotationOrderForFilter=XZY)
    Child=(Type=Bone,Name="offset_skirt_front_lf")
    Filter=(bX=true,bY=true,bZ=True)
    Parents=((Item=(Type=Bone,Name="thigh_twist_01_l"),Weight=1))
    bMaintainOffset=true
    Weight=@n4.Result
   )
   > &RigUnit_RotationConstraintLocalSpaceOffset(
    AdvancedSettings=(InterpolationType=Average,RotationOrderForFilter=XZY)
    Child=(Type=Bone,Name="offset_skirt_front_side_lf")
    Filter=(bX=true,bY=true,bZ=True)
    Parents=((Item=(Type=Bone,Name="thigh_l"),Weight=1))
    bMaintainOffset=true
    Weight=@n4.Result
   )
   > &RigUnit_RotationConstraintLocalSpaceOffset(
    AdvancedSettings=(InterpolationType=Average,RotationOrderForFilter=XZY)
    Child=(Type=Bone,Name="offset_skirt_side_A_01_lf")
    Filter=(bX=true,bY=true,bZ=true)
    Parents=((Item=(Type=Bone,Name="thigh_l"),Weight=1),(Item=(Type=Bone,Name="thigh_l"),Weight=0))
    Weight=0.5
   )
   n5 = &RigVMFunction_AnimEvalRichCurve(
    Curve="(EditorCurveData=(Keys=((InterpMode=RCIM_Cubic,TangentMode=RCTM_Break,TangentWeightMode=RCTWM_WeightedNone,Time=0,Value=0,ArriveTangent=0,ArriveTangentWeight=0,LeaveTangent=0,LeaveTangentWeight=0),(InterpMode=RCIM_Cubic,TangentMode=RCTM_Break,TangentWeight ...(truncated)"
    SourceMaximum=0.5
    SourceMinimum=0
    TargetMinimum=0
    Value=@n2.OutputParam
   )
   > &RigUnit_RotationConstraintLocalSpaceOffset(
    AdvancedSettings=(InterpolationType=Average,RotationOrderForFilter=XZY)
    Child=(Type=Bone,Name="offset_skirt_front_back_lf")
    Filter=(bX=true,bY=true,bZ=true)
    Parents=((Item=(Type=Bone,Name="thigh_l"),Weight=1))
    Weight=@n5.Result
   )
   > &RigUnit_RotationConstraintLocalSpaceOffset(
    AdvancedSettings=(InterpolationType=Average,RotationOrderForFilter=XZY)
    Child=(Type=Bone,Name="offset_skirt_back_lf")
    Filter=(bX=true,bY=true,bZ=true)
    Parents=((Item=(Type=Bone,Name="thigh_l"),Weight=1))
    Weight=@n5.Result
   )
   > &RigUnit_RotationConstraintLocalSpaceOffset(
    AdvancedSettings=(InterpolationType=Average,RotationOrderForFilter=XZY)
    Child=(Type=Bone,Name="offset_skirt_side_B_01_lf")
    Filter=(bX=false,bY=true,bZ=true)
    Parents=((Item=(Type=Bone,Name="thigh_l"),Weight=1))
    Weight=@n5.Result
   )
   > n6 = &RigUnit_SphericalPoseReader(
    ActiveRegionScaleFactors=(PositiveWidth=1,NegativeWidth=0.7,PositiveHeight=1,NegativeHeight=1)
    ActiveRegionSize=0.2
    Debug=(bDrawDebug=false,bDraw2D=False,bDrawLocalAxes=False,DebugScale=25,DebugSegments=20,DebugThickness=0.25)
    DriverAxis=(X=-1,Y=0,Z=0)
    DriverItem=(Type=Bone,Name="thigh_l")
    FalloffRegionScaleFactors=(PositiveWidth=0.5,NegativeWidth=0.9,PositiveHeight=1,NegativeHeight=0.7)
    FalloffSize=0.65
    FlipHeightScaling=False
    FlipWidthScaling=false
    OptionalParentItem=(Type=Bone,Name="None")
    RotationOffset=(X=180,Y=20,Z=0)
   )
   > &RigUnit_ModifyTransforms(
    ItemToModify="((Item=(Type=Bone,Name=\"offset_skirt_side_A_01_lf\"),Transform=(Rotation=(X=0,Y=-0.4226,Z=0,W=0.9063),Translation=(X=0,Y=0,Z=0),Scale3D=(X=1,Y=1,Z=1))),(Item=(Type=Bone,Name=\"offset_skirt_side_B_01_lf\"),Transform=(Rotation=(X=0,Y=-0.5,Z=0,W=0.866),Translati ...(truncated)"
    WeightMinimum=0
    Weight=@n6.OutputParam
   )
   > &RigUnit_ModifyTransforms(
    ItemToModify="((Item=(Type=Bone,Name=\"dyn_skirt_side_A_01_lf\"),Transform=(Rotation=(X=0,Y=0,Z=0.0872,W=0.9962),Translation=(X=0,Y=0,Z=0),Scale3D=(X=1,Y=1,Z=1))),(Item=(Type=Bone,Name=\"dyn_skirt_front_01_lf\"),Transform=(Rotation=(X=0,Y=-0.0872,Z=0,W=0.9962),Translation=( ...(truncated)"
    WeightMinimum=0
    Weight=RigVMDispatchNode(False=0, Condition=RigVMVariableNode() {variable=emote}.Value, True=@n6.OutputParam) {template="DISPATCH_RigVMDispatch_If(in Condition,in True,in False,out Result)"}.Result
   )
   > n7 = &RigUnit_SphericalPoseReader(
    ActiveRegionScaleFactors=(PositiveWidth=1,NegativeWidth=0.7,PositiveHeight=1,NegativeHeight=1.9)
    ActiveRegionSize=0.2
    Debug=(bDrawDebug=false,bDraw2D=false,bDrawLocalAxes=False,DebugScale=25,DebugSegments=20,DebugThickness=0.25)
    DriverAxis=(X=-1,Y=0,Z=0)
    DriverItem=(Type=Bone,Name="thigh_l")
    FalloffRegionScaleFactors=(PositiveWidth=0,NegativeWidth=1,PositiveHeight=1,NegativeHeight=1)
    FalloffSize=0.5
    FlipHeightScaling=False
    FlipWidthScaling=False
    OptionalParentItem=(Type=Bone,Name="None")
    RotationOffset=(X=90,Y=0,Z=-30)
   )
   AlphaInterp = &RigVMFunction_AlphaInterp(InterpSpeedDecreasing=10, InterpSpeedIncreasing=5, Scale=1, bClampResult=False, Value=@n7.OutputParam)
   > &RigUnit_ModifyTransforms(
    ItemToModify="((Item=(Type=Bone,Name=\"dyn_skirt_front_01_lf\"),Transform=(Rotation=(X=-0.1617,Y=-0.3468,Z=-0.3904,W=0.8373),Translation=(X=0.4238,Y=-0.5538,Z=-6.3595),Scale3D=(X=1,Y=1,Z=1))),(Item=(Type=Bone,Name=\"dyn_skirt_front_side_01_lf\"),Transform=(Rotation=(X=-0.02 ...(truncated)"
    WeightMinimum=0
    Weight=@AlphaInterp.Result
   )
   > &RigUnit_ModifyTransforms(
    ItemToModify="((Item=(Type=Bone,Name=\"dyn_skirt_front_01_01_lf\"),Transform=(Rotation=(X=0,Y=0.1305,Z=0,W=0.9914),Translation=(X=-0.2536,Y=0.2536,Z=-4.0994),Scale3D=(X=1,Y=1,Z=1))),(Item=(Type=Bone,Name=\"dyn_skirt_front_side_01_01_lf\"),Transform=(Rotation=(X=0,Y=0,Z=0,W= ...(truncated)"
    WeightMinimum=0
    Weight=@AlphaInterp.Result
   )
   > &RigUnit_ModifyTransforms(
    ItemToModify="((Item=(Type=Bone,Name=\"dyn_skirt_front_02_lf\"),Transform=(Rotation=(X=0,Y=0.2588,Z=0,W=0.9659),Translation=(X=-0.2536,Y=0.2536,Z=-4.0994),Scale3D=(X=1,Y=1,Z=1))),(Item=(Type=Bone,Name=\"dyn_skirt_front_side_02_lf\"),Transform=(Rotation=(X=-0.3432,Y=0.3408,Z ...(truncated)"
    WeightMinimum=0
    Weight=@AlphaInterp.Result
   )
   > &RigUnit_ModifyTransforms(
    ItemToModify="((Item=(Type=Bone,Name=\"dyn_skirt_front_03_lf\"),Transform=(Rotation=(X=0.0029,Y=0.4185,Z=-0.145,W=0.8966),Translation=(X=-0.2536,Y=0.2536,Z=-4.0994),Scale3D=(X=1,Y=1,Z=1))),(Item=(Type=Bone,Name=\"dyn_skirt_front_side_03_lf\"),Transform=(Rotation=(X=-0.0184, ...(truncated)"
    WeightMinimum=0
    Weight=@AlphaInterp.Result
   )
   > Delta Rotation = RigVMFunctionReferenceNode(Child=thigh_l, Parent=pelvis, Source Maximum=20, Source Minimum=0)
   > &RigUnit_ModifyTransforms(
    ItemToModify="((Item=(Type=Bone,Name=\"offset_skirt_front_lf\"),Transform=(Rotation=(X=0,Y=-0.1736,Z=0,W=0.9848),Translation=(X=0,Y=0,Z=0),Scale3D=(X=1,Y=1,Z=1))),(Item=(Type=Bone,Name=\"offset_skirt_front_side_lf\"),Transform=(Rotation=(X=0,Y=-0.1736,Z=0,W=0.9848),Translat ...(truncated)"
    WeightMinimum=0
    Weight=@"Delta Rotation".delta
   )
   > RigVMFunction_Sequence_3 = RigVMAggregateNode()
   RigVMFunction_ControlFlowBranch = &RigVMFunction_ControlFlowBranch(ExecuteContext=@RigVMFunction_Sequence_3.A, Condition=RigVMVariableNode() {variable=frontend}.Value)
   &RigUnit_ModifyTransforms(
    ItemToModify="((Item=(Type=Bone,Name=\"dyn_skirt_front_01_lf\"),Transform=(Rotation=(X=0,Y=-0.1305,Z=0,W=0.9914),Translation=(X=0,Y=0,Z=0),Scale3D=(X=1,Y=1,Z=1))),(Item=(Type=Bone,Name=\"dyn_skirt_front_side_01_lf\"),Transform=(Rotation=(X=0,Y=-0.1305,Z=0,W=0.9914),Translat ...(truncated)"
    WeightMinimum=0
    ExecutePin=@RigVMFunction_ControlFlowBranch.True
    Weight=@n6.OutputParam
   )
   n8 = &RigVMFunction_ControlFlowBranch(ExecuteContext=@RigVMFunction_Sequence_3.B, Condition=RigVMVariableNode() {variable=RBAN}.Value)
   Delta Rotation_1 = RigVMFunctionReferenceNode(Child=calf_l, Parent=thigh_l, Source Maximum=110, ExecuteContext=@n8.True)
   Subtract = &RigVMFunction_MathDoubleSub(
    B=RigVMDispatchNode(BoneName=root, Default=0, Name=deformRig_l_thigh_up_alpha) {template="DISPATCH_RigDispatch_GetAnimAttribute(in Name,in BoneName,in Default,out Value,out Found)"}.Value
    A=@"Delta Rotation_1".delta
   )
   Clamp = &RigVMFunction_MathDoubleClamp(Value=@Subtract.Result)
   > &RigUnit_ModifyTransforms(
    ItemToModify="((Item=(Type=Bone,Name=\"dyn_skirt_back_03_lf\"),Transform=(Rotation=(X=0,Y=-0.1736,Z=0,W=0.9848),Translation=(X=0,Y=0,Z=0),Scale3D=(X=1,Y=1,Z=1))),(Item=(Type=Bone,Name=\"dyn_skirt_front_back_03_lf\"),Transform=(Rotation=(X=0,Y=-0.1736,Z=0,W=0.9848),Translati ...(truncated)"
    WeightMinimum=0
    Weight=@Clamp.Result
   )
   > &RigUnit_RotationConstraintLocalSpaceOffset(
    AdvancedSettings=(InterpolationType=Average,RotationOrderForFilter=XZY)
    Child=(Type=Bone,Name="dyn_belt_back_03_lf")
    Filter=(bX=false,bY=True,bZ=True)
    Parents=((Item=(Type=Bone,Name="dyn_skirt_front_back_02_lf"),Weight=1))
    Weight=1
   )
   n9 = &RigVMFunction_ControlFlowBranch(ExecuteContext=@RigVMFunction_Sequence_3.C, Condition=RigVMVariableNode() {variable=frontend}.Value)
   n10 = &RigUnit_SphericalPoseReader(
    ActiveRegionScaleFactors=(PositiveWidth=1.5,NegativeWidth=1.5,PositiveHeight=1.5,NegativeHeight=1.5)
    ActiveRegionSize=0.2
    Debug=(bDrawDebug=false,bDraw2D=False,bDrawLocalAxes=False,DebugScale=25,DebugSegments=20,DebugThickness=0.25)
    DriverAxis=(X=-1,Y=0,Z=0)
    DriverItem=(Type=Bone,Name="thigh_l")
    FalloffRegionScaleFactors=(PositiveWidth=0.5,NegativeWidth=0.9,PositiveHeight=1,NegativeHeight=0.7)
    FalloffSize=0.65
    FlipHeightScaling=False
    FlipWidthScaling=false
    OptionalParentItem=(Type=Bone,Name="None")
    RotationOffset=(X=0,Y=-10,Z=0)
    ExecutePin=@n9.True
   )
   > &RigUnit_ModifyTransforms(
    ItemToModify="((Item=(Type=Bone,Name=\"offset_skirt_side_A_01_lf\"),Transform=(Rotation=(X=0,Y=-0.5736,Z=0,W=0.8192),Translation=(X=0,Y=0,Z=0),Scale3D=(X=1,Y=1,Z=1))),(Item=(Type=Bone,Name=\"offset_skirt_side_B_01_lf\"),Transform=(Rotation=(X=0,Y=-0.1736,Z=0,W=0.9848),Trans ...(truncated)"
    WeightMinimum=0
    Weight=@n10.OutputParam
   )
   &RigUnit_RotationConstraintLocalSpaceOffset(
    AdvancedSettings=(InterpolationType=Average,RotationOrderForFilter=XZY)
    Child=(Type=Bone,Name="dyn_belt_back_01_lf")
    Filter=(bX=false,bY=True,bZ=True)
    Parents=((Item=(Type=Bone,Name="dyn_skirt_front_back_01_lf"),Weight=1))
    Weight=1
    ExecutePin=@RigVMFunction_Sequence_4.C
   )
   > &RigUnit_RotationConstraintLocalSpaceOffset(
    AdvancedSettings=(InterpolationType=Average,RotationOrderForFilter=XZY)
    Child=(Type=Bone,Name="dyn_belt_front_01_lf")
    Filter=(bX=true,bY=True,bZ=true)
    Parents=((Item=(Type=Bone,Name="dyn_skirt_front_side_01_lf"),Weight=1))
    Weight=0.6
   )
   > &RigUnit_ParentConstraint(
    AdvancedSettings=(RotationOrderForFilter=XZY)
    Child=(Type=Bone,Name="rivet_aim_dyn_belt_front_02_lf")
    Parents="((Item=(Type=Bone,Name=\"dyn_skirt_front_01_01_lf\"),Weight=0.0107),(Item=(Type=Bone,Name=\"dyn_skirt_front_side_01_01_lf\"),Weight=0.0189),(Item=(Type=Bone,Name=\"dyn_skirt_side_A_03_lf\"),Weight=0.0622),(Item=(Type=Bone,Name=\"dyn_skirt_front_03_lf\"),Weight=0.1 ...(truncated)"
    Weight=1
   )
   > &RigUnit_AimConstraintLocalSpaceOffset(
    Child=(Type=Bone,Name="dyn_belt_front_02_lf")
    Filter=(bX=True,bY=True,bZ=True)
    Parents=((Item=(Type=Bone,Name="rivet_aim_dyn_belt_front_02_lf"),Weight=1))
    Weight=1
    WorldUp=(Target=(X=0,Y=0,Z=1),Kind=Direction,Space=(Type=Bone,Name="rivet_aim_dyn_belt_front_02_lf"))
   )
   n11 = &RigVMFunction_ControlFlowBranch(ExecuteContext=@RigVMFunction_Sequence_4.D, Condition=RigVMVariableNode() {variable=frontend}.Value)
   &RigUnit_ModifyTransforms(
    ItemToModify=((Item=(Type=Bone,Name="base_belt_front_lf"),Transform=(Rotation=(X=0,Y=-0.1736,Z=0,W=0.9848),Translation=(X=0,Y=0,Z=0),Scale3D=(X=1,Y=1,Z=1))))
    WeightMinimum=0
    ExecutePin=@n11.True
    Weight=@n10.OutputParam
   )
   n12 = &RigVMFunction_Sequence(ExecuteContext=@RigVMFunction_Sequence_4.B)
   RigUnit_GetRelativeTransformForItem = &RigUnit_GetRelativeTransformForItem(Child=(Type=Bone,Name="pelvis"), Parent=(Type=Bone,Name="root"), bChildInitial=False, bParentInitial=False)
   Remap = &RigVMFunction_MathDoubleRemap(
    SourceMaximum=40
    SourceMinimum=70
    TargetMaximum=0
    TargetMinimum=1
    Value=@RigUnit_GetRelativeTransformForItem.RelativeTransform.Translation.Z
   )
   n13 = &RigVMFunction_AnimEvalRichCurve(
    Curve="(EditorCurveData=(Keys=((InterpMode=RCIM_Cubic,TangentMode=RCTM_Break,TangentWeightMode=RCTWM_WeightedNone,Time=0,Value=0,ArriveTangent=0,ArriveTangentWeight=0,LeaveTangent=0,LeaveTangentWeight=0),(InterpMode=RCIM_Cubic,TangentMode=RCTM_Break,TangentWeight ...(truncated)"
    SourceMinimum=0
    TargetMinimum=0
    Value=@Remap.Result
   )
   n14 = &RigVMFunction_MathBoolToFloat(Value=RigVMVariableNode() {variable=Crouch}.Value)
   Multiply = &RigVMFunction_MathDoubleMul(A=@n13.Result, B=@n14.Result)
   &RigUnit_ModifyTransforms(
    ItemToModify="((Item=(Type=Bone,Name=\"dyn_skirt_front_side_01_lf\"),Transform=(Rotation=(X=0,Y=-0.1736,Z=0,W=0.9848),Translation=(X=0,Y=0,Z=0),Scale3D=(X=1,Y=1,Z=1))),(Item=(Type=Bone,Name=\"dyn_skirt_side_B_01_lf\"),Transform=(Rotation=(X=0,Y=0,Z=-0.1305,W=0.9914),Transla ...(truncated)"
    WeightMinimum=0
    Weight=@Multiply.Result
    ExecutePin=@n12.A
   )
   n15 = &RigVMFunction_AnimEvalRichCurve(
    Curve="(EditorCurveData=(Keys=((InterpMode=RCIM_Cubic,TangentMode=RCTM_Break,TangentWeightMode=RCTWM_WeightedNone,Time=0,Value=0,ArriveTangent=0,ArriveTangentWeight=0,LeaveTangent=0,LeaveTangentWeight=0),(InterpMode=RCIM_Cubic,TangentMode=RCTM_Break,TangentWeight ...(truncated)"
    SourceMinimum=0
    TargetMaximum=0
    TargetMinimum=1
    Value=@Remap.Result
   )
   &RigUnit_ModifyTransforms(
    ItemToModify="((Item=(Type=Bone,Name=\"dyn_skirt_side_B_01_lf\"),Transform=(Rotation=(X=0,Y=0,Z=0,W=1),Translation=(X=0,Y=0,Z=0),Scale3D=(X=1,Y=1,Z=1))),(Item=(Type=Bone,Name=\"offset_skirt_front_back_lf\"),Transform=(Rotation=(X=0,Y=-0.0872,Z=0,W=0.9962),Translation=(X=0,Y ...(truncated)"
    WeightMinimum=0
    Weight=&RigVMFunction_MathDoubleMul(A=@n15.Result, B=@n14.Result).Result
    ExecutePin=@n12.B
   )
   RigVMFunctionReturnNode(ExecuteContext=@RigVMFunction_Sequence.B)
   &RigVMFunction_AnimEvalRichCurve(
    Curve="(EditorCurveData=(Keys=((InterpMode=RCIM_Cubic,TangentMode=RCTM_Break,TangentWeightMode=RCTWM_WeightedNone,Time=0,Value=0,ArriveTangent=0,ArriveTangentWeight=0,LeaveTangent=0,LeaveTangentWeight=0),(InterpMode=RCIM_Cubic,TangentMode=RCTM_Break,TangentWeight ...(truncated)"
    SourceMaximum=0.5
    SourceMinimum=0
    TargetMinimum=0
    Value=@n2.OutputParam
   )
   RigVMVariableNode() {variable=emote}
  ]}
 New Function_ContainedGraph {nodes [
   n1 = RigVMFunctionEntryNode()
   RigVMFunction_Sequence = &RigVMFunction_Sequence(ExecuteContext=@n1.ExecuteContext)
   RigVMFunction_Sequence_1 = RigVMAggregateNode(ExecuteContext=@RigVMFunction_Sequence.A)
   Delta Rotation_1 = RigVMFunctionReferenceNode(Child=neck_02, Parent=spine_05, Source Maximum=-45, Source Minimum=-15, ExecuteContext=@RigVMFunction_Sequence_1.A)
   > &RigUnit_ModifyTransforms(
    ItemToModify="((Item=(Type=Bone,Name=\"base_neck_back_cr\"),Transform=(Rotation=(X=0,Y=0,Z=-0.7071,W=0.7071),Translation=(X=0,Y=-2,Z=0),Scale3D=(X=1,Y=1,Z=1))),(Item=(Type=Bone,Name=\"dyn_neck_back_side_rt\"),Transform=(Rotation=(X=0,Y=0,Z=-0.2588,W=0.9659),Translation=(X=0 ...(truncated)"
    WeightMinimum=0
    Weight=@"Delta Rotation_1".delta
   )
   > Delta Rotation_2 = RigVMFunctionReferenceNode(Child=head, Parent=neck_02, Source Maximum=-45, Source Minimum=-15)
   > &RigUnit_ModifyTransforms(
    ItemToModify="((Item=(Type=Bone,Name=\"base_neck_back_cr\"),Transform=(Rotation=(X=0,Y=0,Z=-0.0872,W=0.9962),Translation=(X=-2,Y=0.1062,Z=0.0042),Scale3D=(X=1,Y=1,Z=1))),(Item=(Type=Bone,Name=\"dyn_neck_back_side_rt\"),Transform=(Rotation=(X=0,Y=0,Z=-0.0872,W=0.9962),Transl ...(truncated)"
    WeightMinimum=0
    Weight=@"Delta Rotation_2".delta
   )
   > Delta Rotation_3 = RigVMFunctionReferenceNode(Child=neck_01, Parent=spine_05, Source Maximum=40, Source Minimum=15)
   > &RigUnit_ModifyTransforms(
    ItemToModify="((Item=(Type=Bone,Name=\"base_neck_front_cr\"),Transform=(Rotation=(X=0,Y=0,Z=0.1736,W=0.9848),Translation=(X=0,Y=0,Z=0),Scale3D=(X=1,Y=1,Z=1))),(Item=(Type=Bone,Name=\"base_neck_front_side_lf\"),Transform=(Rotation=(X=0,Y=0,Z=0.1736,W=0.9848),Translation=(X=0 ...(truncated)"
    WeightMinimum=0
    Weight=@"Delta Rotation_3".delta
   )
   n2 = &RigUnit_SphericalPoseReader(
    ActiveRegionScaleFactors=(PositiveWidth=0.5,NegativeWidth=1,PositiveHeight=1.5,NegativeHeight=1.5)
    ActiveRegionSize=0.2
    Debug=(bDrawDebug=false,bDraw2D=False,bDrawLocalAxes=False,DebugScale=25,DebugSegments=20,DebugThickness=0.25)
    DriverAxis=(X=1,Y=0,Z=0)
    DriverItem=(Type=Bone,Name="clavicle_l")
    FalloffRegionScaleFactors=(PositiveWidth=0.6,NegativeWidth=1,PositiveHeight=1,NegativeHeight=1)
    FalloffSize=0.4
    FlipHeightScaling=false
    FlipWidthScaling=false
    OptionalParentItem=(Type=Bone,Name="None")
    RotationOffset=(X=0,Y=-35,Z=0)
    ExecutePin=@RigVMFunction_Sequence_1.B
   )
   > &RigUnit_ModifyTransforms(
    ItemToModify="((Item=(Type=Bone,Name=\"base_neck_side_lf\"),Transform=(Rotation=(X=0.0436,Y=-0.0436,Z=0.0019,W=0.9981),Translation=(X=-0.0396,Y=0.1319,Z=6.6806),Scale3D=(X=1,Y=1,Z=1))),(Item=(Type=Bone,Name=\"base_neck_back_side_lf\"),Transform=(Rotation=(X=0,Y=0.0436,Z=0,W ...(truncated)"
    WeightMinimum=0
    Weight=@n2.OutputParam
   )
   > n3 = &RigUnit_SphericalPoseReader(
    ActiveRegionScaleFactors=(PositiveWidth=0.5,NegativeWidth=1,PositiveHeight=1.5,NegativeHeight=1.5)
    ActiveRegionSize=0.2
    Debug=(bDrawDebug=false,bDraw2D=False,bDrawLocalAxes=False,DebugScale=25,DebugSegments=20,DebugThickness=0.25)
    DriverAxis=(X=-1,Y=0,Z=0)
    DriverItem=(Type=Bone,Name="clavicle_r")
    FalloffRegionScaleFactors=(PositiveWidth=0.6,NegativeWidth=1,PositiveHeight=1,NegativeHeight=1)
    FalloffSize=0.4
    FlipHeightScaling=false
    FlipWidthScaling=false
    OptionalParentItem=(Type=Bone,Name="None")
    RotationOffset=(X=0,Y=155,Z=0)
   )
   > &RigUnit_ModifyTransforms(
    ItemToModify="((Item=(Type=Bone,Name=\"base_neck_side_rt\"),Transform=(Rotation=(X=0.0057,Y=-0.1304,Z=-0.0432,W=0.9905),Translation=(X=-0.1342,Y=-0.0178,Z=-7.7611),Scale3D=(X=1,Y=1,Z=1))),(Item=(Type=Bone,Name=\"base_neck_back_side_rt\"),Transform=(Rotation=(X=0,Y=0.0436,Z= ...(truncated)"
    WeightMinimum=0
    Weight=@n3.OutputParam
   )
   > n4 = &RigUnit_SphericalPoseReader(
    ActiveRegionScaleFactors=(PositiveWidth=0.5,NegativeWidth=1,PositiveHeight=1.5,NegativeHeight=1.5)
    ActiveRegionSize=0.2
    Debug=(bDrawDebug=false,bDraw2D=False,bDrawLocalAxes=False,DebugScale=25,DebugSegments=20,DebugThickness=0.25)
    DriverAxis=(X=0,Y=1,Z=0)
    DriverItem=(Type=Bone,Name="head")
    FalloffRegionScaleFactors=(PositiveWidth=0.6,NegativeWidth=1,PositiveHeight=1,NegativeHeight=1)
    FalloffSize=0.4
    FlipHeightScaling=false
    FlipWidthScaling=false
    OptionalParentItem=(Type=Bone,Name="None")
    RotationOffset=(X=0,Y=155,Z=0)
   )
   > &RigUnit_ModifyTransforms(
    ItemToModify=((Item=(Type=Bone,Name="base_neck_side_lf"),Transform=(Rotation=(X=0,Y=0.1305,Z=0,W=0.9914),Translation=(X=0,Y=0,Z=0),Scale3D=(X=1,Y=1,Z=1))))
    WeightMinimum=0
    Weight=@n4.OutputParam
   )
   n5 = &RigUnit_SphericalPoseReader(
    ActiveRegionScaleFactors=(PositiveWidth=1,NegativeWidth=1,PositiveHeight=1,NegativeHeight=1)
    ActiveRegionSize=0.15
    Debug=(bDrawDebug=false,bDraw2D=False,bDrawLocalAxes=False,DebugScale=25,DebugSegments=20,DebugThickness=0.25)
    DriverAxis=(X=-1,Y=0,Z=0)
    DriverItem=(Type=Bone,Name="upperarm_r")
    FalloffRegionScaleFactors=(PositiveWidth=1,NegativeWidth=1,PositiveHeight=1,NegativeHeight=1)
    FalloffSize=0.3
    FlipHeightScaling=false
    FlipWidthScaling=false
    OptionalParentItem=(Type=Bone,Name="None")
    RotationOffset=(X=-121.2053,Y=-2.3134,Z=28.751)
    ExecutePin=@RigVMFunction_Sequence_1.D
   )
   > &RigUnit_ModifyTransforms(
    ItemToModify="((Item=(Type=Bone,Name=\"base_neck_side_rt\"),Transform=(Rotation=(X=0,Y=-0.1736,Z=0,W=0.9848),Translation=(X=0,Y=-0.062,Z=-2.1186),Scale3D=(X=1,Y=1,Z=1))),(Item=(Type=Bone,Name=\"base_neck_front_side_rt\"),Transform=(Rotation=(X=0.1305,Y=0,Z=0,W=0.9914),Trans ...(truncated)"
    WeightMinimum=0
    Weight=@n5.OutputParam
   )
   > n6 = &RigUnit_SphericalPoseReader(
    ActiveRegionScaleFactors=(PositiveWidth=1,NegativeWidth=1,PositiveHeight=1,NegativeHeight=1)
    ActiveRegionSize=0.01
    Debug=(bDrawDebug=false,bDraw2D=False,bDrawLocalAxes=False,DebugScale=25,DebugSegments=20,DebugThickness=0.25)
    DriverAxis=(X=0,Y=1,Z=0)
    DriverItem=(Type=Bone,Name="head")
    FalloffRegionScaleFactors=(PositiveWidth=1,NegativeWidth=1,PositiveHeight=1,NegativeHeight=1)
    FalloffSize=0.2
    FlipHeightScaling=False
    FlipWidthScaling=False
    OptionalParentItem=(Type=Bone,Name="clavicle_r")
    RotationOffset=(X=0,Y=-87,Z=-5)
   )
   > &RigUnit_ModifyTransforms(
    ItemToModify="((Item=(Type=Bone,Name=\"base_neck_back_cr\"),Transform=(Rotation=(X=0,Y=0,Z=-0.4226,W=0.9063),Translation=(X=-2.179,Y=2.0569,Z=0),Scale3D=(X=1,Y=1,Z=1))),(Item=(Type=Bone,Name=\"base_neck_back_side_rt\"),Transform=(Rotation=(X=0,Y=0,Z=-0.6088,W=0.7934),Transl ...(truncated)"
    WeightMinimum=0
    Weight=@n6.OutputParam
   )
   RigVMFunctionReturnNode(ExecuteContext=@RigVMFunction_Sequence.B)
  ]}
 New Function_ContainedGraph {nodes [
   n1 = RigVMFunctionEntryNode()
   RigVMFunction_MathQuaternionToEuler = &RigVMFunction_MathQuaternionToEuler(
    RotationOrder=ZYX
    Value=&RigUnit_GetRelativeTransformForItem(
     Child=(Type=Bone,Name="hand_l")
     Parent=(Type=Bone,Name="lowerarm_l")
     bChildInitial=False
     bParentInitial=False
     Child.Name=@n1.Child
     Parent.Name=@n1.Parent
    ).RelativeTransform.Rotation
   )
   Remap = &RigVMFunction_MathDoubleRemap(
    TargetMaximum=1
    TargetMinimum=0
    Value=@RigVMFunction_MathQuaternionToEuler.Result.Z
    SourceMaximum=@n1.Source Maximum
    SourceMinimum=@n1.Source Minimum
   )
   > RigVMFunctionReturnNode(delta=@Remap.Result)
  ]}
 New Function_ContainedGraph {nodes [
   n1 = RigVMFunctionEntryNode()
   RigVMFunction_Sequence = &RigVMFunction_Sequence(ExecuteContext=@n1.ExecuteContext)
   RigUnit_GetTransform = &RigUnit_GetTransform(Item=(Type=Bone,Name="thigh_l_for_cons"), Space=LocalSpace, bInitial=false, Item.Name=@n1.Bone)
   n2 = &RigVMFunction_MathQuaternionSwingTwist(TwistAxis=(X=1,Y=0,Z=0), Input=@RigUnit_GetTransform.Transform.Rotation)
   n3 = &RigVMFunction_MathQuaternionToEuler(RotationOrder=ZXY, Value=@n2.Swing)
   GreaterEqual = /Script/RigVM.RigVMFunction_MathDoubleGreaterEqual(B=0, A=@n3.Result.Z) {is=struct RigVMUnitNode, template="GreaterEqual::Execute(in A,in B,out Result)"}
   n4 = RigVMDispatchNode(
    Condition=@n1.Front Z
    False=RigVMDispatchNode(
     False=0
     Condition=/Script/RigVM.RigVMFunction_MathDoubleLessEqual(B=0, A=@n3.Result.Z) {is=struct RigVMUnitNode, template="LessEqual::Execute(in A,in B,out Result)"}.Result
     True=@n3.Result.Z
    ) {template="DISPATCH_RigVMDispatch_If(in Condition,in True,in False,out Result)"}.Result
    True=RigVMDispatchNode(False=0, Condition=@GreaterEqual.Result, True=@n3.Result.Z) {template="DISPATCH_RigVMDispatch_If(in Condition,in True,in False,out Result)"}.Result
   ) {template="DISPATCH_RigVMDispatch_If(in Condition,in True,in False,out Result)"}
   RigVMFunction_MathQuaternionToEuler = &RigVMFunction_MathQuaternionToEuler(RotationOrder=ZXY, Value=@n2.Swing)
   LessEqual = &RigVMFunction_MathDoubleLessEqual(B=0, A=@RigVMFunction_MathQuaternionToEuler.Result.Y) {template="LessEqual::Execute(in A,in B,out Result)"}
   n5 = RigVMDispatchNode(
    Condition=@n1.Front Y
    False=RigVMDispatchNode(False=0, Condition=@LessEqual.Result, True=@RigVMFunction_MathQuaternionToEuler.Result.Y) {template="DISPATCH_RigVMDispatch_If(in Condition,in True,in False,out Result)"}.Result
    True=RigVMDispatchNode(
     False=0
     Condition=&RigVMFunction_MathDoubleGreaterEqual(B=0, A=@RigVMFunction_MathQuaternionToEuler.Result.Y) {template="GreaterEqual::Execute(in A,in B,out Result)"}.Result
     True=@RigVMFunction_MathQuaternionToEuler.Result.Y
    ) {template="DISPATCH_RigVMDispatch_If(in Condition,in True,in False,out Result)"}.Result
   ) {template="DISPATCH_RigVMDispatch_If(in Condition,in True,in False,out Result)"}
   Remap = &RigVMFunction_MathDoubleRemap(SourceMinimum=-10, TargetMinimum=0, SourceMaximum=@n5.Result, TargetMaximum=@n5.Result, Value=@n5.Result)
   RigVMFunction_MathQuaternionFromEuler = &RigVMFunction_MathQuaternionFromEuler(
    Euler=(X=0,Y=0,Z=0)
    RotationOrder=ZXY
    Euler.Z=@n4.Result
    Euler.Y=@Remap.Result
    Euler.X=&RigVMFunction_MathQuaternionToEuler(RotationOrder=ZXY, Value=@n2.Twist).Result.X
   )
   RigVMFunctionReturnNode(RotOut=@RigVMFunction_MathQuaternionFromEuler.Result, ExecuteContext=@RigVMFunction_Sequence.B)
   &RigVMFunction_MathDoubleGreaterEqual(A=0, B=0) {template="GreaterEqual::Execute(in A,in B,out Result)"}
   RigVMDispatchNode(Condition=False, False=0, True=@n4.Result) {template="DISPATCH_RigVMDispatch_If(in Condition,in True,in False,out Result)"}
   &RigVMFunction_MathQuaternionSwingTwist(TwistAxis=(X=1,Y=0,Z=0), Input=@n2.Twist)
  ]}
 New Function_ContainedGraph {nodes [
   RigVMFunctionEntryNode()
   > &RigUnit_ParentConstraint(
    AdvancedSettings=(RotationOrderForFilter=XZY)
    Child=(Type=Bone,Name="base_skirt_front_lf")
    Parents=((Item=(Type=Bone,Name="spine_03"),Weight=0.0242),(Item=(Type=Bone,Name="spine_04"),Weight=0.0602),(Item=(Type=Bone,Name="spine_02"),Weight=0.2596),(Item=(Type=Bone,Name="pelvis"),Weight=0.656))
    Weight=1
   )
   > &RigUnit_ParentConstraint(
    AdvancedSettings=(RotationOrderForFilter=XZY)
    Child=(Type=Bone,Name="base_skirt_side_A_01_lf")
    Parents=((Item=(Type=Bone,Name="spine_03"),Weight=0.0191),(Item=(Type=Bone,Name="spine_01"),Weight=0.0194),(Item=(Type=Bone,Name="spine_04"),Weight=0.0475),(Item=(Type=Bone,Name="spine_02"),Weight=0.2884),(Item=(Type=Bone,Name="pelvis"),Weight=0.6256))
    Weight=1
   )
   > &RigUnit_ParentConstraint(
    AdvancedSettings=(RotationOrderForFilter=XZY)
    Child=(Type=Bone,Name="base_skirt_front_side_lf")
    Parents=((Item=(Type=Bone,Name="spine_01"),Weight=0.023),(Item=(Type=Bone,Name="spine_03"),Weight=0.0305),(Item=(Type=Bone,Name="spine_04"),Weight=0.0718),(Item=(Type=Bone,Name="spine_02"),Weight=0.3051),(Item=(Type=Bone,Name="pelvis"),Weight=0.5696))
    Weight=1
   )
   > &RigUnit_ParentConstraint(
    AdvancedSettings=(RotationOrderForFilter=XZY)
    Child=(Type=Bone,Name="base_skirt_side_B_01_lf")
    Parents=((Item=(Type=Bone,Name="spine_03"),Weight=0.0116),(Item=(Type=Bone,Name="spine_04"),Weight=0.0299),(Item=(Type=Bone,Name="spine_01"),Weight=0.0389),(Item=(Type=Bone,Name="spine_02"),Weight=0.1611),(Item=(Type=Bone,Name="pelvis"),Weight=0.7585))
    Weight=1
   )
   > &RigUnit_ParentConstraint(
    AdvancedSettings=(RotationOrderForFilter=XZY)
    Child=(Type=Bone,Name="base_skirt_front_back_lf")
    Parents=((Item=(Type=Bone,Name="spine_03"),Weight=0.0146),(Item=(Type=Bone,Name="spine_04"),Weight=0.0306),(Item=(Type=Bone,Name="spine_01"),Weight=0.0612),(Item=(Type=Bone,Name="spine_02"),Weight=0.1518),(Item=(Type=Bone,Name="pelvis"),Weight=0.7417))
    Weight=1
   )
   > &RigUnit_ParentConstraint(
    AdvancedSettings=(RotationOrderForFilter=XZY)
    Child=(Type=Bone,Name="base_skirt_back_lf")
    Parents=((Item=(Type=Bone,Name="spine_04"),Weight=0.0218),(Item=(Type=Bone,Name="spine_01"),Weight=0.0272),(Item=(Type=Bone,Name="spine_02"),Weight=0.0612),(Item=(Type=Bone,Name="dyn_skirt_back_01_lf"),Weight=0.1763),(Item=(Type=Bone,Name="pelvis"),Weight=0.7135))
    Weight=1
   )
   > RigVMFunctionReturnNode()
   &RigUnit_ParentConstraint(
    AdvancedSettings=(RotationOrderForFilter=XZY)
    Child=(Type=Bone,Name="rivet_aim_dyn_belt_front_02_lf")
    ExecutePin=()
    Parents="((Item=(Type=Bone,Name=\"dyn_skirt_front_01_01_lf\"),Weight=0.0107),(Item=(Type=Bone,Name=\"dyn_skirt_front_side_01_01_lf\"),Weight=0.0189),(Item=(Type=Bone,Name=\"dyn_skirt_side_A_03_lf\"),Weight=0.0622),(Item=(Type=Bone,Name=\"dyn_skirt_front_03_lf\"),Weight=0.1 ...(truncated)"
    Weight=1
   )
  ]}
 RigVMFunctionLibrary {nodes [
   Controls = RigVMCollapseNode(Debug=False, ExecuteContext=())
   Delta Rotation = RigVMCollapseNode(ExecuteContext=())
   Surface = RigVMCollapseNode(ExecuteContext=())
   neck = RigVMCollapseNode(ExecuteContext=())
   skirt = RigVMCollapseNode(ExecuteContext=())
   skirt_rotation = RigVMCollapseNode(ExecuteContext=(), Twist Axis Z=())
  ]}
 RigVMFunction_Sequence_1_ContainedGraph {nodes [
   n1 = RigVMFunctionEntryNode()
   RigVMFunction_Sequence = &RigVMFunction_Sequence(ExecuteContext=@n1.ExecuteContext)
   n2 = &RigVMFunction_Sequence(ExecuteContext=@RigVMFunction_Sequence.B)
   n3 = &RigVMFunction_Sequence(ExecuteContext=@n2.B)
   RigVMFunctionReturnNode(A=@RigVMFunction_Sequence.A, B=@n2.A, C=@n3.A, D=@n3.B)
  ]}
 RigVMFunction_Sequence_1_ContainedGraph {nodes [
   n1 = RigVMFunctionEntryNode()
   n2 = &RigVMFunction_Sequence(ExecuteContext=@n1.ExecuteContext)
   n3 = &RigVMFunction_Sequence(ExecuteContext=@n2.B)
   n4 = &RigVMFunction_Sequence(ExecuteContext=@n3.B)
   RigVMFunctionReturnNode(B=@n3.A, C=@n4.A, D=@n4.B, A=@n2.A)
  ]}
 RigVMFunction_Sequence_1_ContainedGraph {nodes [
   n1 = RigVMFunctionEntryNode()
   n2 = &RigVMFunction_Sequence(ExecuteContext=@n1.ExecuteContext)
   n3 = &RigVMFunction_Sequence(ExecuteContext=@n2.B)
   RigVMFunctionReturnNode(B=@n3.A, C=@n3.B, A=@n2.A)
  ]}
 RigVMFunction_Sequence_2_ContainedGraph {nodes [
   n1 = RigVMFunctionEntryNode()
   n2 = &RigVMFunction_Sequence(ExecuteContext=@n1.ExecuteContext)
   n3 = &RigVMFunction_Sequence(ExecuteContext=@n2.B)
   n4 = &RigVMFunction_Sequence(ExecuteContext=@n3.B)
   RigVMFunctionReturnNode(A=@n2.A, B=@n3.A, C=@n4.A, D=@n4.B)
  ]}
 RigVMModel {nodes [
   RigUnit_BeginExecution = &RigUnit_BeginExecution()
   RigVMFunction_Sequence = RigVMAggregateNode(ExecuteContext=@RigUnit_BeginExecution.ExecutePin)
   Controls = RigVMFunctionReferenceNode(Debug=false, ExecuteContext=@RigVMFunction_Sequence.A)
   FN_DeformRig_FNC = RigVMFunctionReferenceNode(
    DeformRig_ Curve="(deformRig_l_upperarm_up_alpha_72_#1=False,deformRig_r_upperarm_up_alpha_73_#2=False,deformRig_l_upperarm_fr_alpha_74_#3=False,deformRig_r_upperarm_fr_alpha_75_CB51B8 ...(truncated)"
    DeformRig_Bool="(upperarm_1_#4=True,elbow_3_#5=True,wrist_7_#6=True,groin_8_#7=True,glute_17_#8=True,knee_9_40C408324CA261 ...(truncated)"
    DeformRig_Debug=(upperarm_debug_18_#4=False,wrist_debug_20_#6=False,lat_debug_26_#9=False,groinGlute_debug_27_#7=False)
    DeformRig_Value="(upperarm_up_18_#10=(Rotation=(X=0.0,Y=-0.0,Z=0.0,W=1.0),Translation=(X=0.0,Y=0.0,Z=8.0),Scale3D=(X=0.70,Y=1.0,Z=1.0)),upperarm_fwd_fr_19_#11=(Rotation=(X=0.0,Y=-0.0,Z=0.0,W=1.0),Translation=(X=0.0, ...(truncated)"
    ExecuteContext=@RigVMFunction_Sequence.B
   )
   Surface = RigVMFunctionReferenceNode(ExecuteContext=@RigVMFunction_Sequence.C)
   skirt = RigVMFunctionReferenceNode(ExecuteContext=@RigVMFunction_Sequence.D)
   > neck = RigVMFunctionReferenceNode()
  ]}
]
hierarchy [
 root {loc=<0 0 0>, rot=<0 0 0 1>}
 pelvis {parent=root, loc=<0 -0.5301 98.0571>, rot=<0.0127 0.707 -0.0127 -0.707>}
 spine_01 {parent=pelvis, loc=<6.1208 -0.9328 0>, rot=<0 0 0.034 -0.9994>}
 spine_02 {parent=spine_01, loc=<6.5997 -0.2095 0>, rot=<0 0 0.0238 -0.9997>}
 spine_03 {parent=spine_02, loc=<6.4914 -0.1237 0>, rot=<0 0 0.0149 -0.9999>}
 spine_04 {parent=spine_03, loc=<6.9598 0.0794 0>, rot=<0 0 -0.0132 -0.9999>}
 spine_05 {parent=spine_04, loc=<6.9987 0 0>, rot=<0 0 -0.0015 -1>}
 attach_cape {parent=spine_05, loc=<9.0258 -11.7071 -0.003>, rot=<0.0616 -0.7045 -0.0612 -0.7044>}
 neck_01 {parent=spine_05, loc=<11.3138 0.1529 0>, rot=<0 0 -0.2003 -0.9797>}
 neck_02 {parent=neck_01, loc=<4.5627 0 0>, rot=<0 0 0.0504 -0.9987>}
 head {parent=neck_02, loc=<5.3914 0 0>, rot=<0 0 0.115 -0.9934>}
 attach_backpack {parent=spine_05, loc=<-0.1124 -13.4603 -0.0003>, rot=<-0.0092 -0.7071 0.0097 -0.707>}
 clavicle_l {parent=spine_05, loc=<5.4831 -0.1053 -3.2946>, rot=<0.0465 -0.7289 -0.009 -0.6829>}
 upperarm_l {parent=clavicle_l, loc=<9.8658 -1.2136 -0.4105>, rot=<0.0164 -0.3245 -0.0269 -0.9454>}
 lowerarm_l {parent=upperarm_l, loc=<26.465 -0.8345 0.259>, rot=<0.0036 0.0161 -0.2298 -0.9731>}
 hand_l {parent=lowerarm_l, loc=<22.7355 -0.3203 -0.6105>, rot=<-0.6784 -0.0223 0.0735 -0.7307>}
 thumb_01_l {parent=hand_l, loc=<1.3118 -0.3191 -2.459>, rot=<0.6546 -0.4042 0.0664 -0.6354>}
 thumb_02_l {parent=thumb_01_l, loc=<5.9336 -0.0448 0.0401>, rot=<-0.0037 0.0213 0.298 -0.9543>}
 thumb_03_l {parent=thumb_02_l, loc=<2.2223 0.3682 -0.0907>, rot=<0.0118 -0.013 0.0066 -0.9998>}
 index_metacarpal_l {parent=hand_l, loc=<1.6535 0.0183 -2.96>, rot=<0.0158 -0.0124 -0.0173 -0.9996>}
 index_01_l {parent=index_metacarpal_l, loc=<8.3008 0.0357 -0.0524>, rot=<-0.0044 0.0485 0.0924 -0.9945>}
 index_02_l {parent=index_01_l, loc=<3.3587 -0.3004 0.0452>, rot=<-0.002 0.0027 0.3225 -0.9466>}
 index_03_l {parent=index_02_l, loc=<2.5578 -0.123 -0.0252>, rot=<-0.002 0.0067 0.2558 -0.9667>}
 middle_metacarpal_l {parent=hand_l, loc=<2.1856 -0.1725 -0.5984>, rot=<0.006 0.0168 -0.0329 -0.9993>}
 middle_01_l {parent=middle_metacarpal_l, loc=<7.3275 -0.2447 0.0129>, rot=<-0.0208 0.0042 0.0858 -0.9961>}
 middle_02_l {parent=middle_01_l, loc=<4.0123 0.1401 0.1175>, rot=<-0.0218 0.0258 0.3512 -0.9357>}
 middle_03_l {parent=middle_02_l, loc=<2.6077 0.0534 0.1373>, rot=<0.0082 0.0188 0.1708 -0.9851>}
 ring_metacarpal_l {parent=hand_l, loc=<2.2074 -0.3625 1.3638>, rot=<-0.0284 0.0722 -0.0458 -0.9959>}
 ring_01_l {parent=ring_metacarpal_l, loc=<7.0295 -0.6197 -0.1446>, rot=<-0.0489 -0.0252 0.1892 -0.9804>}
 ring_02_l {parent=ring_01_l, loc=<3.6174 0.4647 -0.1027>, rot=<0 0 0.2806 -0.9598>}
 ring_03_l {parent=ring_02_l, loc=<2.4152 0.114 0.0199>, rot=<0 0 0.1189 -0.9929>}
 pinky_metacarpal_l {parent=hand_l, loc=<1.7616 -0.9753 3.3814>, rot=<-0.0397 0.0814 -0.069 -0.9935>}
 pinky_01_l {parent=pinky_metacarpal_l, loc=<7.1709 -0.4831 -0.1094>, rot=<-0.0654 -0.0035 0.2242 -0.9723>}
 pinky_02_l {parent=pinky_01_l, loc=<2.8274 0.0893 -0.1043>, rot=<0.0037 0.0011 0.2854 -0.9584>}
 pinky_03_l {parent=pinky_02_l, loc=<2.2046 0.0307 -0.0078>, rot=<-0.0002 0.0098 0.1445 -0.9895>}
 deform_wrist_dn_base_l {parent=hand_l, loc=<0 -1.5326 0>, rot=<0 0 0 -1>}
 deform_wrist_dn_l {parent=deform_wrist_dn_base_l, loc=<0 0 0>, rot=<0 0 0 1>}
 deform_wrist_up_base_l {parent=hand_l, loc=<0 2.3009 0>, rot=<0 0 0 -1>}
 deform_wrist_up_l {parent=deform_wrist_up_base_l, loc=<0 0 0>, rot=<0 0 0 1>}
 lowerarm_twist_02_l {parent=lowerarm_l, loc=<10.6707 0 0>, rot=<0 0 0 1>}
 lowerarm_twist_01_l {parent=lowerarm_l, loc=<18.316 0 0>, rot=<0 0 0 1>}
 deform_elbow_in_base_l {parent=lowerarm_l, loc=<0.1834 3.6062 -0.2572>, rot=<0 0 0 -1>}
 deform_elbow_in_l {parent=deform_elbow_in_base_l, loc=<0 0 0>, rot=<0 0 0 1>}
 deform_elbow_out_base_l {parent=lowerarm_l, loc=<0 -4 0>, rot=<0 0 0 1>}
 deform_elbow_out_l {parent=deform_elbow_out_base_l, loc=<0 0 0>, rot=<0 0 0 1>}
 upperarm_twist_01_l {parent=upperarm_l, loc=<7.2135 -0.1134 0>, rot=<0 0 0 1>}
 deform_upperarm_bk_base_l {parent=upperarm_twist_01_l, loc=<-6.4053 -4.3746 0.314>, rot=<0 0 0 -1>}
 deform_upperarm_bk_l {parent=deform_upperarm_bk_base_l, loc=<0 0 0>, rot=<0 0 0 1>}
 deform_upperarm_fr_base_l {parent=upperarm_twist_01_l, loc=<-5.6312 5.092 0.3354>, rot=<0 0 0 -1>}
 deform_upperarm_fr_l {parent=deform_upperarm_fr_base_l, loc=<0 0 0>, rot=<0 0 0 1>}
 deform_upperarm_up_base_l {parent=upperarm_twist_01_l, loc=<-8.891 0.4825 3.9517>, rot=<0 0 0 -1>}
 deform_upperarm_up_l {parent=deform_upperarm_up_base_l, loc=<0 0 0>, rot=<0 0 0 1>}
 upperarm_twist_02_l {parent=upperarm_l, loc=<19.7998 -0.1134 0>, rot=<0 0 0 1>}
 clavicle_r {parent=spine_05, loc=<5.4831 -0.1053 3.2946>, rot=<0.7289 0.0465 -0.6829 0.009>}
 upperarm_r {parent=clavicle_r, loc=<-9.8658 1.2136 0.4105>, rot=<0.0164 -0.3245 -0.0269 -0.9454>}
 lowerarm_r {parent=upperarm_r, loc=<-26.465 0.8345 -0.259>, rot=<0.0036 0.0161 -0.2298 -0.9731>}
 hand_r {parent=lowerarm_r, loc=<-22.7355 0.3203 0.6106>, rot=<-0.6784 -0.0223 0.0735 -0.7307>}
 thumb_01_r {parent=hand_r, loc=<-1.3118 0.3191 2.459>, rot=<0.6546 -0.4042 0.0664 -0.6354>}
 thumb_02_r {parent=thumb_01_r, loc=<-5.9336 0.0448 -0.0401>, rot=<-0.0037 0.0213 0.298 -0.9543>}
 thumb_03_r {parent=thumb_02_r, loc=<-2.2223 -0.3682 0.0907>, rot=<0.0118 -0.013 0.0066 -0.9998>}
 index_metacarpal_r {parent=hand_r, loc=<-1.6535 -0.0183 2.96>, rot=<0.0158 -0.0124 -0.0173 -0.9996>}
 index_01_r {parent=index_metacarpal_r, loc=<-8.3008 -0.0357 0.0524>, rot=<-0.0044 0.0485 0.0924 -0.9945>}
 index_02_r {parent=index_01_r, loc=<-3.3587 0.3004 -0.0452>, rot=<-0.002 0.0027 0.3225 -0.9466>}
 index_03_r {parent=index_02_r, loc=<-2.5578 0.123 0.0252>, rot=<-0.002 0.0067 0.2558 -0.9667>}
 middle_metacarpal_r {parent=hand_r, loc=<-2.1856 0.1725 0.5984>, rot=<0.006 0.0168 -0.0329 -0.9993>}
 middle_01_r {parent=middle_metacarpal_r, loc=<-7.3275 0.2447 -0.0129>, rot=<-0.0208 0.0042 0.0858 -0.9961>}
 middle_02_r {parent=middle_01_r, loc=<-4.0123 -0.1401 -0.1175>, rot=<-0.0218 0.0258 0.3512 -0.9357>}
 middle_03_r {parent=middle_02_r, loc=<-2.6077 -0.0534 -0.1373>, rot=<0.0082 0.0188 0.1708 -0.9851>}
 ring_metacarpal_r {parent=hand_r, loc=<-2.2074 0.3625 -1.3638>, rot=<-0.0284 0.0722 -0.0458 -0.9959>}
 ring_01_r {parent=ring_metacarpal_r, loc=<-7.0295 0.6197 0.1446>, rot=<-0.0489 -0.0252 0.1892 -0.9804>}
 ring_02_r {parent=ring_01_r, loc=<-3.6174 -0.4647 0.1027>, rot=<0 0 0.2806 -0.9598>}
 ring_03_r {parent=ring_02_r, loc=<-2.4152 -0.114 -0.0199>, rot=<0 0 0.1189 -0.9929>}
 pinky_metacarpal_r {parent=hand_r, loc=<-1.7616 0.9753 -3.3814>, rot=<-0.0397 0.0814 -0.069 -0.9935>}
 pinky_01_r {parent=pinky_metacarpal_r, loc=<-7.1709 0.4831 0.1094>, rot=<-0.0654 -0.0035 0.2242 -0.9723>}
 pinky_02_r {parent=pinky_01_r, loc=<-2.8274 -0.0893 0.1043>, rot=<0.0037 0.0011 0.2854 -0.9584>}
 pinky_03_r {parent=pinky_02_r, loc=<-2.2046 -0.0307 0.0078>, rot=<-0.0002 0.0098 0.1445 -0.9895>}
 deform_wrist_dn_base_r {parent=hand_r, loc=<-0.0001 1.5328 0>, rot=<0 0 0 -1>}
 deform_wrist_dn_r {parent=deform_wrist_dn_base_r, loc=<0 0 0>, rot=<0 0 0 1>}
 deform_wrist_up_base_r {parent=hand_r, loc=<-0.0002 -2.3007 -0.0001>, rot=<0 0 0 -1>}
 deform_wrist_up_r {parent=deform_wrist_up_base_r, loc=<0 0 0>, rot=<0 0 0 1>}
 lowerarm_twist_02_r {parent=lowerarm_r, loc=<-10.6707 0 0>, rot=<0 0 0 -1>}
 lowerarm_twist_01_r {parent=lowerarm_r, loc=<-18.316 0 0>, rot=<0 0 0 -1>}
 deform_elbow_in_base_r {parent=lowerarm_r, loc=<-0.1834 -3.6062 0.2572>, rot=<0 0 0 -1>}
 deform_elbow_in_r {parent=deform_elbow_in_base_r, loc=<0 0 0>, rot=<0 0 0 1>}
 deform_elbow_out_base_r {parent=lowerarm_r, loc=<0 4 -0.0001>, rot=<0 0 0 -1>}
 deform_elbow_out_r {parent=deform_elbow_out_base_r, loc=<0 0 0>, rot=<0 0 0 1>}
 upperarm_twist_01_r {parent=upperarm_r, loc=<-7.2135 0.1134 0>, rot=<0 0 0 -1>}
 deform_upperarm_bk_base_r {parent=upperarm_twist_01_r, loc=<6.4053 4.3746 -0.314>, rot=<0 0 0 -1>}
 deform_upperarm_bk_r {parent=deform_upperarm_bk_base_r, loc=<0 0 0>, rot=<0 0 0 1>}
 deform_upperarm_fr_base_r {parent=upperarm_twist_01_r, loc=<5.6312 -5.092 -0.3354>, rot=<0 0 0 -1>}
 deform_upperarm_fr_r {parent=deform_upperarm_fr_base_r, loc=<0 0 0>, rot=<0 0 0 1>}
 deform_upperarm_up_base_r {parent=upperarm_twist_01_r, loc=<8.891 -0.4825 -3.9517>, rot=<0 0 0 -1>}
 deform_upperarm_up_r {parent=deform_upperarm_up_base_r, loc=<0 0 0>, rot=<0 0 0 1>}
 upperarm_twist_02_r {parent=upperarm_r, loc=<-19.7998 0.1134 0>, rot=<0 0 0 -1>}
 base_neck_side_lf {parent=spine_05, loc=<8.7213 0.1255 -9.5151>, rot=<0.0037 -0.0914 -0.0353 -0.9952>}
 dyn_neck_side_lf {parent=base_neck_side_lf, loc=<0 0 0>, rot=<0 -0.0292 0.0015 -0.9996>}
 base_neck_front_cr {parent=spine_05, loc=<6.0333 7.1498 0.007>, rot=<0 0.0011 -0.04 -0.9992>}
 dyn_neck_front_cr {parent=base_neck_front_cr, loc=<0 0 0>, rot=<0 0.0003 0 -1>}
 base_neck_front_side_lf {parent=spine_05, loc=<7.0584 4.8851 -6.2019>, rot=<0.2752 -0.0141 -0.0444 -0.9603>}
 dyn_neck_front_side_lf {parent=base_neck_front_side_lf, loc=<0 0 0>, rot=<0 -0.0058 -0.0032 -1>}
 base_neck_back_side_lf {parent=spine_05, loc=<10.2536 -5.824 -7.106>, rot=<-0.272 -0.0602 -0.0217 -0.9602>}
 dyn_neck_back_side_lf {parent=base_neck_back_side_lf, loc=<0 0 0>, rot=<0 -0.0118 0.0082 -0.9999>}
 base_neck_front_side_rt {parent=spine_05, loc=<7.0003 4.869 6.2357>, rot=<-0.2754 0.0167 -0.0444 -0.9602>}
 dyn_neck_front_side_rt {parent=base_neck_front_side_rt, loc=<0 0 0>, rot=<0 0.0063 -0.0033 -1>}
 base_neck_back_side_rt {parent=spine_05, loc=<10.2626 -5.8228 7.0973>, rot=<0.272 0.0587 -0.0222 -0.9602>}
 dyn_neck_back_side_rt {parent=base_neck_back_side_rt, loc=<0 0 0>, rot=<0 0.0112 0.0078 -0.9999>}
 base_neck_side_rt {parent=spine_05, loc=<8.9466 0.1507 9.3765>, rot=<-0.0023 0.0568 -0.0372 -0.9977>}
 dyn_neck_side_rt {parent=base_neck_side_rt, loc=<0 0 0>, rot=<0 0.0057 0.0003 -1>}
 base_neck_back_cr {parent=spine_05, loc=<11.15 -7.5441 -0.0301>, rot=<0.0001 -0.0015 -0.0396 -0.9992>}
 dyn_neck_back_cr {parent=base_neck_back_cr, loc=<0 0 0>, rot=<0 -0.0003 0.0001 -1>}
 deform_pec_aim_base_l {parent=spine_04, loc=<3.2456 9.3309 -1.2237>, rot=<0.0603 -0.4033 0.3318 -0.8507>}
 deform_pec_aim_l {parent=deform_pec_aim_base_l, loc=<0 0 0>, rot=<0 0 0 1>}
 deform_pec_base_l {parent=deform_pec_aim_l, loc=<9.3665 0 0>, rot=<0 0 0 1>}
 deform_pec_l {parent=deform_pec_base_l, loc=<0 0 0>, rot=<0 0 0 1>}
 deform_pec_aim_base_r {parent=spine_04, loc=<3.2452 9.3309 1.2237>, rot=<-0.4033 -0.0603 0.8507 0.3318>}
 deform_pec_aim_r {parent=deform_pec_aim_base_r, loc=<0 0 0>, rot=<0 0 0 1>}
 deform_pec_base_r {parent=deform_pec_aim_r, loc=<-9.3667 -0.0003 -0.0004>, rot=<0 0 0 1>}
 deform_pec_r {parent=deform_pec_base_r, loc=<0 0 0>, rot=<0 0 0 1>}
 thigh_l {parent=pelvis, loc=<-1.0855 -0.2109 -8.7053>, rot=<0.0009 0.0355 0.0155 -0.9993>}
 calf_l {parent=thigh_l, loc=<-42.5284 -0.0019 -0.7297>, rot=<0.0006 -0.0161 -0.0391 -0.9991>}
 foot_l {parent=calf_l, loc=<-46.7401 -2.5014 -0.592>, rot=<0.0427 -0.0238 0.0353 -0.9982>}
 ball_l {parent=foot_l, loc=<-3.5658 10.8066 -0.1544>, rot=<0.0001 -0.0001 -0.7063 -0.7079>}
 calf_twist_02_l {parent=calf_l, loc=<-19.1546 -1.7943 -0.2638>, rot=<0.0237 0.0095 -0.0035 -0.9997>}
 calf_twist_01_l {parent=calf_l, loc=<-35.848 -1.92 -0.579>, rot=<-0.0193 0.0343 -0.0378 -0.9985>}
 base_laces_lf {parent=calf_twist_01_l, loc=<-4.7072 6.0901 -0.5664>, rot=<-0.555 -0.493 -0.4226 -0.5199>}
 base_laces_01_lf {parent=base_laces_lf, loc=<0 0 0>, rot=<0 0 0.7071 -0.7071>}
 dyn_laces_01_lf {parent=base_laces_01_lf, loc=<0 0 0>, rot=<0 0 0 1>}
 base_laces_02_lf {parent=base_laces_lf, loc=<0 0 0>, rot=<0 0 -0.7071 -0.7071>}
 dyn_laces_02_lf {parent=base_laces_02_lf, loc=<0 0 0>, rot=<0 0 0 1>}
 deform_knee_in_base_l {parent=calf_l, loc=<0.8031 -5.0231 0.2456>, rot=<0 0 0 -1>}
 deform_knee_in_l {parent=deform_knee_in_base_l, loc=<0 0 0>, rot=<0 0 0 1>}
 deform_knee_out_base_l {parent=calf_l, loc=<0 4 0>, rot=<0 0 0 -1>}
 deform_knee_out_l {parent=deform_knee_out_base_l, loc=<0 0 0>, rot=<0 0 0 1>}
 thigh_twist_01_l {parent=thigh_l, loc=<-27.952 0 0>, rot=<0.0013 0 0 -1>}
 deform_groin_base_l {parent=thigh_l, loc=<0.4801 5.6776 -0.2486>, rot=<0 0 0 -1>}
 deform_groin_l {parent=deform_groin_base_l, loc=<0 0 0>, rot=<0 0 0 1>}
 tihgh_l_cor {parent=thigh_l, loc=<0 0 0>, rot=<0 0 0 1>}
 thigh_r {parent=pelvis, loc=<-1.0855 -0.2109 8.7053>, rot=<0.0355 -0.0009 0.9993 0.0155>}
 calf_r {parent=thigh_r, loc=<42.5284 0.0019 0.7297>, rot=<0.0006 -0.0161 -0.0391 -0.9991>}
 foot_r {parent=calf_r, loc=<46.7401 2.5014 0.592>, rot=<0.0427 -0.0238 0.0353 -0.9982>}
 ball_r {parent=foot_r, loc=<3.5658 -10.8066 0.1544>, rot=<0.0001 -0.0001 -0.7063 -0.7079>}
 calf_twist_02_r {parent=calf_r, loc=<19.1546 1.7943 0.2638>, rot=<0.0237 0.0095 -0.0035 -0.9997>}
 calf_twist_01_r {parent=calf_r, loc=<35.8483 1.9197 0.5794>, rot=<0.0237 0.0095 -0.0035 -0.9997>}
 base_laces_rt {parent=calf_twist_01_r, loc=<5.0981 -5.6464 0.3569>, rot=<-0.4748 0.4936 0.5492 -0.479>}
 base_laces_01_rt {parent=base_laces_rt, loc=<0 0 0>, rot=<0.003 -0.0393 0.7373 -0.6744>}
 dyn_laces_01_rt {parent=base_laces_01_rt, loc=<0 0 0>, rot=<0 0 0 1>}
 base_laces_02_rt {parent=base_laces_rt, loc=<0 0 0>, rot=<0 0 -0.7071 -0.7071>}
 dyn_laces_02_rt {parent=base_laces_02_rt, loc=<0 0 0>, rot=<0 0 0 1>}
 deform_knee_in_base_r {parent=calf_r, loc=<-0.8031 5.0231 -0.2457>, rot=<0 0 0 -1>}
 deform_knee_in_r {parent=deform_knee_in_base_r, loc=<0 0 0>, rot=<0 0 0 1>}
 deform_knee_out_base_r {parent=calf_r, loc=<0 -4 0>, rot=<0 0 0 -1>}
 deform_knee_out_r {parent=deform_knee_out_base_r, loc=<0 0 0>, rot=<0 0 0 1>}
 thigh_twist_01_r {parent=thigh_r, loc=<27.952 0 0>, rot=<0.0013 0 0 -1>}
 deform_groin_base_r {parent=thigh_r, loc=<-0.4801 -5.6776 0.2486>, rot=<0 0 0 -1>}
 deform_groin_r {parent=deform_groin_base_r, loc=<0 0 0>, rot=<0 0 0 1>}
 base_skirt_front_lf {parent=pelvis, loc=<8.836 8.7225 2.1282>, rot=<0 -0.7071 -0.7071 0>}
 offset_skirt_front_lf {parent=base_skirt_front_lf, loc=<0 0 0>, rot=<0 0 0 1>}
 dyn_skirt_front_01_lf {parent=offset_skirt_front_lf, loc=<0 0 0>, rot=<0 0 0 1>}
 dyn_skirt_front_01_01_lf {parent=dyn_skirt_front_01_lf, loc=<11.8825 -1.4664 1.6284>, rot=<-0.108 0.0691 0.0527 -0.9904>}
 dyn_skirt_front_02_lf {parent=dyn_skirt_front_01_01_lf, loc=<10.1912 0 0>, rot=<0 0 0 1>}
 dyn_skirt_front_03_lf {parent=dyn_skirt_front_02_lf, loc=<16.7999 0 0>, rot=<0 0 0 1>}
 dyn_skirt_front_04_lf {parent=dyn_skirt_front_03_lf, loc=<16.7999 0 0>, rot=<0 0 0 1>}
 base_skirt_front_side_lf {parent=pelvis, loc=<9.2878 5.696 -7.3035>, rot=<0 -0.7071 -0.7071 0>}
 offset_skirt_front_side_lf {parent=base_skirt_front_side_lf, loc=<0 0 0>, rot=<0 0 0 1>}
 dyn_skirt_front_side_01_lf {parent=offset_skirt_front_side_lf, loc=<0 0 0>, rot=<0 0 0 1>}
 dyn_skirt_front_side_01_01_lf {parent=dyn_skirt_front_side_01_lf, loc=<11.8177 -2.413 1.752>, rot=<-0.3819 0.0779 0.0531 -0.9194>}
 dyn_skirt_front_side_02_lf {parent=dyn_skirt_front_side_01_01_lf, loc=<10.231 0 0>, rot=<0 0 0 1>}
 dyn_skirt_front_side_03_lf {parent=dyn_skirt_front_side_02_lf, loc=<16.5954 0 0>, rot=<0 0 0 1>}
 dyn_skirt_front_side_04_lf {parent=dyn_skirt_front_side_03_lf, loc=<15.9377 0 0>, rot=<0 0 0 1>}
 base_skirt_back_lf {parent=pelvis, loc=<6.2134 -9.8233 5.212>, rot=<0 -0.7071 0.7071 0>}
 offset_skirt_back_lf {parent=base_skirt_back_lf, loc=<0 0 0>, rot=<0 0 0 1>}
 dyn_skirt_back_01_lf {parent=offset_skirt_back_lf, loc=<0 0 0>, rot=<0 0 0 1>}
 dyn_skirt_back_02_lf {parent=dyn_skirt_back_01_lf, loc=<19.4327 2.1792 2.9234>, rot=<0.0211 0.0503 -0.0533 -0.9971>}
 dyn_skirt_back_03_lf {parent=dyn_skirt_back_02_lf, loc=<16.7999 0 0>, rot=<0 0 0 1>}
 dyn_skirt_back_04_lf {parent=dyn_skirt_back_03_lf, loc=<16.7999 0 0>, rot=<0 0 0 1>}
 base_skirt_front_back_lf {parent=pelvis, loc=<8.5098 -7.186 -7.6295>, rot=<0 -0.7071 0.7071 0>}
 offset_skirt_front_back_lf {parent=base_skirt_front_back_lf, loc=<0 0 0>, rot=<0 0 0 1>}
 dyn_skirt_front_back_01_lf {parent=offset_skirt_front_back_lf, loc=<0 0 0>, rot=<0 0 0 1>}
 dyn_skirt_front_back_02_lf {parent=dyn_skirt_front_back_01_lf, loc=<20.7441 4.1552 2.8789>, rot=<0.2912 0.0709 -0.067 -0.9517>}
 dyn_skirt_front_back_03_lf {parent=dyn_skirt_front_back_02_lf, loc=<16.7999 0 0>, rot=<0 0 0 1>}
 dyn_skirt_front_back_04_lf {parent=dyn_skirt_front_back_03_lf, loc=<16.7999 0 0>, rot=<0 0 0 1>}
 deform_glute_base_l {parent=pelvis, loc=<-7.3466 -9.0213 -9.1245>, rot=<0.0172 0.031 -0.4477 -0.8935>}
 deform_glute_l {parent=deform_glute_base_l, loc=<0 0 0>, rot=<0 0 0 1>}
 deform_glute_base_r {parent=pelvis, loc=<-7.3464 -9.0213 9.1245>, rot=<0.031 -0.0172 0.8935 -0.4477>}
 deform_glute_r {parent=deform_glute_base_r, loc=<0 0 0>, rot=<0 0 0 1>}
 base_belt_back_lf {parent=pelvis, loc=<9.6651 -6.7558 -7.9094>, rot=<0.1037 -0.8805 0.3625 -0.2873>}
 dyn_belt_back_01_lf {parent=base_belt_back_lf, loc=<0 0 0>, rot=<0 0 0 1>}
 dyn_belt_back_02_lf {parent=dyn_belt_back_01_lf, loc=<3.3787 0 0>, rot=<0.0214 -0.0458 -0.0147 -0.9986>}
 dyn_belt_back_03_lf {parent=dyn_belt_back_02_lf, loc=<3.3116 0 0>, rot=<-0.038 -0.1711 0.0153 -0.9844>}
 dyn_belt_back_04_lf {parent=dyn_belt_back_03_lf, loc=<8.6853 0 0>, rot=<0.0022 0.0078 -0.0005 -1>}
 dyn_belt_back_05_lf {parent=dyn_belt_back_04_lf, loc=<7.1313 0.022 0.074>, rot=<0.0085 0.0038 -0.002 -1>}
 base_belt_front_lf {parent=pelvis, loc=<10.0376 5.1312 -7.0577>, rot=<-0.1091 -0.9053 -0.2894 -0.2912>}
 dyn_belt_front_01_lf {parent=base_belt_front_lf, loc=<0 0 0>, rot=<0 0 0 1>}
 dyn_belt_front_02_lf {parent=dyn_belt_front_01_lf, loc=<4.2735 0 0>, rot=<0.0024 -0.1555 -0.0012 -0.9878>}
 dyn_belt_front_03_lf {parent=dyn_belt_front_02_lf, loc=<3.3467 0 0>, rot=<0.0811 -0.0384 0.0017 -0.996>}
 dyn_belt_front_04_lf {parent=dyn_belt_front_03_lf, loc=<8.4283 0 0>, rot=<0 0 0 1>}
 dyn_belt_front_05_lf {parent=dyn_belt_front_04_lf, loc=<7.162 0 0>, rot=<0 0 0 1>}
 base_belt_back_rt {parent=pelvis, loc=<9.6826 -6.7369 7.8852>, rot=<0.2715 -0.4006 0.8679 -0.1118>}
 dyn_belt_back_01_rt {parent=base_belt_back_rt, loc=<0 0 0>, rot=<0 0 0 1>}
 dyn_belt_back_02_rt {parent=dyn_belt_back_01_rt, loc=<3.7503 0 0>, rot=<0.0162 -0.1 -0.0093 -0.9948>}
 dyn_belt_back_03_rt {parent=dyn_belt_back_02_rt, loc=<3.3021 0 0>, rot=<0.0066 -0.0562 -0.0025 -0.9984>}
 dyn_belt_back_04_rt {parent=dyn_belt_back_03_rt, loc=<5.4727 0 0>, rot=<0 -0.0206 0 -0.9998>}
 dyn_belt_back_05_rt {parent=dyn_belt_back_04_rt, loc=<4.941 0 0>, rot=<0 0 0 1>}
 dyn_bottle_rt {parent=dyn_belt_back_02_rt, loc=<-0.5281 0 0.9229>, rot=<0 0 0 1>}
 thigh_l_for_cons_offset {parent=pelvis, loc=<-1.0859 -0.2119 -8.688>, rot=<0.0009 0.0355 0.0155 -0.9993>}
 thigh_l_for_cons_set_rotation {parent=thigh_l_for_cons_offset, loc=<0 0 0>, rot=<0 0 0 1>}
 thigh_l_for_cons {parent=thigh_l_for_cons_offset, loc=<0 0 0>, rot=<0 0 0 1>}
 base_skirt_side_A_01_lf {parent=pelvis, loc=<9.0467 0.6569 -9.7357>, rot=<0 -1 0 0>}
 offset_skirt_side_A_01_lf {parent=base_skirt_side_A_01_lf, loc=<0 0 0>, rot=<0 0 0 1>}
 dyn_skirt_side_A_01_lf {parent=offset_skirt_side_A_01_lf, loc=<0 0 0>, rot=<0 0 0 1>}
 dyn_skirt_side_A_02_lf {parent=dyn_skirt_side_A_01_lf, loc=<20.555 2.4163 5.7676>, rot=<-0.0166 0.119 -0.045 -0.9917>}
 dyn_skirt_side_A_03_lf {parent=dyn_skirt_side_A_02_lf, loc=<17.036 0 0>, rot=<0 0 0 1>}
 dyn_skirt_side_A_04_lf {parent=dyn_skirt_side_A_03_lf, loc=<17.036 0 0>, rot=<0 0 0 1>}
 base_skirt_side_B_01_lf {parent=pelvis, loc=<9.0467 -4.5729 -9.7357>, rot=<0 -1 0 0>}
 offset_skirt_side_B_01_lf {parent=base_skirt_side_B_01_lf, loc=<0 0 0>, rot=<0 0 0 1>}
 dyn_skirt_side_B_01_lf {parent=offset_skirt_side_B_01_lf, loc=<0 0 0>, rot=<0 0 0 1>}
 dyn_skirt_side_B_02_lf {parent=dyn_skirt_side_B_01_lf, loc=<20.6578 0.4091 6.1148>, rot=<0.2008 0.0893 0.0644 -0.9734>}
 dyn_skirt_side_B_03_lf {parent=dyn_skirt_side_B_02_lf, loc=<17.036 0 0>, rot=<0 0 0 1>}
 dyn_skirt_side_B_04_lf {parent=dyn_skirt_side_B_03_lf, loc=<17.036 0 0>, rot=<0 0 0 1>}
 rivet_aim_dyn_belt_front_02_lf {parent=pelvis, loc=<-23.8803 11.0628 -11.3523>, rot=<-0.0631 -0.9387 -0.3053 -0.1469>}
 thigh_l_ctrl {controlType=EULER_TRANSFORM, type=Control, loc=<0 0 0>, rot=<0 0 0 1>}
 calf_l_ctrl {controlType=EULER_TRANSFORM, parent=thigh_l_ctrl, type=Control, loc=<0 0 0>, rot=<0 0 0 1>}
 thigh_r_ctrl {controlType=EULER_TRANSFORM, type=Control, loc=<0 0 0>, rot=<0 0 0 1>}
 calf_r_ctrl {controlType=EULER_TRANSFORM, parent=thigh_r_ctrl, type=Control, loc=<0 0 0>, rot=<0 0 0 1>}
 clavicle_l_ctrl {controlType=EULER_TRANSFORM, type=Control, loc=<0 0 0>, rot=<0 0 0 1>}
 clavicle_r_ctrl {controlType=EULER_TRANSFORM, type=Control, loc=<0 0 0>, rot=<0 0 0 1>}
 neck_01_ctrl {controlType=EULER_TRANSFORM, type=Control, loc=<0 0 0>, rot=<0 0 0 1>}
 neck_02_ctrl {controlType=EULER_TRANSFORM, parent=neck_01_ctrl, type=Control, loc=<0 0 0>, rot=<0 0 0 1>}
 head_ctrl {controlType=EULER_TRANSFORM, parent=neck_02_ctrl, type=Control, loc=<0 0 0>, rot=<0 0 0 1>}
]
variables [
 Crouch {type=bool, default=False}
 Input {type=FQuat, default=(X=0,Y=0,Z=0,W=1)}
 Moving {type=double, default=0}
 RBAN {type=bool, default=False}
 Skydive {type=bool, default=False}
 emote {type=bool, default=False}
 frontend {type=bool, default=False}
]
