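# Physics asset for the ChemPencil medium female cosmetic body, built around the left-side (_lf) skirt bones.
# Capsules on pelvis, thighs and calves form the body, and an oversized root capsule sits offset from it.
# Six skirt chains (back, front, front-back, front-side, side A, side B) hang from the pelvis, some through tiny offset bodies.
# Several upper joints move freely, while lower joints have narrow swing limits and mostly zero twist.
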PhysicsAsset /BRCosmetics/Characters/Player/Female/Medium/Bodies/F_MED_ChemPencil/Meshes/F_MED_ChemPencil_Physics
bodies {
 calf_l [{shapeType=Capsule, length=54.0328, radius=8.1851, center=<-27.2719 -3.1375 0>, rotation=<rot -90 -166.7373 166.7373>}]
 calf_r [{shapeType=Capsule, length=54.5674, radius=10.2832, center=<29.1612 3.1375 0>, rotation=<rot -90 -90 -90>}]
 dyn_skirt_back_01_lf [{
   shapeType=Capsule
   length=7.3644
   radius=3.2622
   shapeName=dyn_skirt_back_01_lf_capsule
   center=<7.0832 3.4844 0.8148>
   rotation=<rot 0 94.4802 -87.0054>
  }]
 dyn_skirt_back_02_lf [{
   shapeType=Capsule
   length=9.5267
   radius=4.9274
   shapeName=dyn_skirt_back_02_lf_capsule
   center=<3.9438 4.5704 0>
   rotation=<rot 90 172.875 -97.125>
  }]
 dyn_skirt_back_03_lf [{
   shapeType=Capsule
   length=10.9343
   radius=6.6819
   shapeName=dyn_skirt_back_03_lf_capsule
   center=<6.4554 5.9232 0>
   rotation=<rot 90 175.2364 85.2363>
  }]
 dyn_skirt_back_04_lf [{
   shapeType=Capsule
   length=24.6166
   radius=9.7683
   shapeName=dyn_skirt_back_04_lf_capsule
   center=<6.7671 7.5967 0.8277>
   rotation=<rot 90 -165.9638 109.0362>
  }]
 dyn_skirt_front_02_lf [{
   shapeType=Capsule
   length=12.7175
   radius=1.6151
   shapeName=dyn_skirt_front_02_lf_capsule
   center=<4.8738 -3.6172 0>
   rotation=<rot 90 180 90>
  }]
 dyn_skirt_front_03_lf [{
   shapeType=Capsule
   length=15.1793
   radius=2.6899
   shapeName=dyn_skirt_front_03_lf_capsule
   center=<7.8169 0 0>
   rotation=<rot 85.0005 5.0002 -175.0002>
  }]
 dyn_skirt_front_04_lf [{
   shapeType=Capsule
   length=21.2657
   radius=2.9815
   shapeName=dyn_skirt_front_04_lf_capsule
   center=<7.5481 -10.6897 -0.8112>
   rotation=<rot 85 -90.0005 -0.0006>
  }]
 dyn_skirt_front_back_01_lf [{
   shapeType=Capsule
   length=7.9402
   radius=5.1276
   shapeName=dyn_skirt_front_back_01_lf_capsule
   center=<12.2842 -1.5513 -1.0856>
   rotation=<rot -2.5945 -82.0168 -95.2635>
  }]
 dyn_skirt_front_back_02_lf [{
   shapeType=Capsule
   length=9.0887
   radius=3.7067
   shapeName=dyn_skirt_front_back_02_lf_capsule
   center=<3.9417 -0.0613 -1.9577>
   rotation=<rot 90 -165.9638 109.0361>
  }]
 dyn_skirt_front_back_03_lf [{
   shapeType=Capsule
   length=6.2687
   radius=4.4599
   shapeName=dyn_skirt_front_back_03_lf_capsule
   center=<2 0 0>
   rotation=<rot 90 -11.6164 -101.6166>
  }]
 dyn_skirt_front_back_04_lf [{
   shapeType=Capsule
   length=26.0696
   radius=10.3966
   shapeName=dyn_skirt_front_back_04_lf_capsule
   center=<7.2644 0 1.7913>
   rotation=<rot 90 180 89.9998>
  }]
 dyn_skirt_front_side_02_lf [{
   shapeType=Capsule
   length=4.9839
   radius=3.523
   shapeName=dyn_skirt_front_side_02_lf_capsule
   center=<1.2282 -0.4713 -5.3868>
   rotation=<rot 85 90.0014 -179.9988>
  }]
 dyn_skirt_front_side_03_lf [{
   shapeType=Capsule
   length=12.5222
   radius=3.4185
   shapeName=dyn_skirt_front_side_03_lf_capsule
   center=<6.335 0.054 0.4383>
   rotation=<rot 80 90.0007 -179.9995>
  }]
 dyn_skirt_front_side_04_lf [{
   shapeType=Capsule
   length=19.1042
   radius=4.3161
   shapeName=dyn_skirt_front_side_04_lf_capsule
   center=<5.4922 0.2131 1.0412>
   rotation=<rot 75 90.0007 -179.9996>
  }]
 dyn_skirt_side_A_01_lf [{
   shapeType=Capsule
   length=5.9332
   radius=3.1538
   shapeName=dyn_skirt_side_A_01_lf_capsule
   center=<19.0121 -0.0574 4.4128>
   rotation=<rot -14.8009 -173.2555 -106.3246>
  }]
 dyn_skirt_side_A_02_lf [{
   shapeType=Capsule
   length=5.3782
   radius=3.6588
   shapeName=dyn_skirt_side_A_02_lf_capsule
   center=<7.1395 1.1117 -1.1695>
   rotation=<rot 0 -83.8835 -85.3374>
  }]
 dyn_skirt_side_A_03_lf [{
   shapeType=Capsule
   length=10.8821
   radius=1.2952
   shapeName=dyn_skirt_side_A_03_lf_capsule
   center=<5 0 0>
   rotation=<rot -24.9999 90.0002 90>
  }]
 dyn_skirt_side_A_04_lf [{shapeType=Capsule, length=7.1378, radius=5, shapeName=dyn_skirt_side_A_04_lf_capsule, center=<5 0 0>, rotation=<rot 0 0.0001 125>}]
 dyn_skirt_side_B_01_lf [{
   shapeType=Capsule
   length=9.2748
   radius=1.1419
   shapeName=dyn_skirt_side_B_01_lf_capsule
   center=<9.7022 -0.2354 3.2399>
   rotation=<rot 0 84.0092 -75.8261>
  }]
 dyn_skirt_side_B_02_lf [{
   shapeType=Capsule
   length=5.303
   radius=2.5199
   shapeName=dyn_skirt_side_B_02_lf_capsule
   center=<6.971 0.8241 0.714>
   rotation=<rot 0 -93.7833 -88.5206>
  }]
 dyn_skirt_side_B_03_lf [{shapeType=Capsule, length=8.255, radius=4.5735, shapeName=dyn_skirt_side_B_03_lf_capsule, center=<4.9765 0 0>, rotation=<rot 0 0.0001 40>}]
 dyn_skirt_side_B_04_lf [{
   shapeType=Capsule
   length=18.7047
   radius=4.7919
   shapeName=dyn_skirt_side_B_04_lf_capsule
   center=<7.0005 -2.101 -1.955>
   rotation=<rot 0.0001 0.0001 35>
  }]
 offset_skirt_back_lf [{shapeType=Capsule, length=0.505, radius=0.505, shapeName=offset_skirt_back_lf_capsule, rotation=<rot 0 -90 0>}]
 offset_skirt_side_A_01_lf [{shapeType=Capsule, length=0.505, radius=0.505, shapeName=offset_skirt_side_A_01_lf_capsule, rotation=<rot 0 -90 0>}]
 offset_skirt_side_B_01_lf [{shapeType=Capsule, length=0.505, radius=0.505, shapeName=offset_skirt_side_B_01_lf_capsule, rotation=<rot 0 -90 0>}]
 pelvis [{shapeType=Capsule, length=7.7503, radius=8.5582, center=<0 0.0493 0>, rotation=<rot 0 -90 0>}]
 root [{
   shapeType=Capsule
   length=105.5291
   radius=249.3996
   shapeName=root_capsule
   center=<3.473 -19.6961 -249.9252>
   rotation=<rot 90 -178.8065 171.1933>
  }]
 thigh_l [{shapeType=Capsule, length=29.5506, radius=7.3313, center=<-28.8534 -0.2781 1.2325>, rotation=<rot -0.4308 -88.4701 -94.9307>}]
 thigh_r [{shapeType=Capsule, length=20.9269, radius=7.875, center=<11.3525 2.1914 -1.4757>, rotation=<rot -0.4308 -88.4701 85.0693>}]
}
constraints [
 {bone1Name=thigh_l, bone2Name=pelvis, swing1LimitDegrees=45, swing2LimitDegrees=45, twistLimitDegrees=45}
 {bone1Name=calf_l, bone2Name=thigh_l, swing1LimitDegrees=45, swing2LimitDegrees=45, twistLimitDegrees=45}
 {bone1Name=thigh_r, bone2Name=pelvis, swing1LimitDegrees=45, swing2LimitDegrees=45, twistLimitDegrees=45}
 {bone1Name=calf_r, bone2Name=thigh_r, swing1LimitDegrees=45, swing2LimitDegrees=45, twistLimitDegrees=45}
 {
  bone1Name=offset_skirt_back_lf
  bone2Name=pelvis
  swing1LimitDegrees=45
  swing1Motion=Free
  swing2LimitDegrees=45
  swing2Motion=Free
  twistLimitDegrees=45
  twistMotion=Free
 }
 {bone1Name=dyn_skirt_back_01_lf, bone2Name=offset_skirt_back_lf, swing1LimitDegrees=10, swing2LimitDegrees=10, twistLimitDegrees=0}
 {bone1Name=dyn_skirt_back_02_lf, bone2Name=dyn_skirt_back_01_lf, swing1LimitDegrees=10, swing2LimitDegrees=15, twistLimitDegrees=0}
 {bone1Name=dyn_skirt_back_03_lf, bone2Name=dyn_skirt_back_02_lf, swing1LimitDegrees=10, swing2LimitDegrees=30, twistLimitDegrees=0}
 {bone1Name=dyn_skirt_back_04_lf, bone2Name=dyn_skirt_back_03_lf, swing1LimitDegrees=10, swing2LimitDegrees=20, twistLimitDegrees=0}
 {
  bone1Name=dyn_skirt_front_back_01_lf
  bone2Name=pelvis
  swing1LimitDegrees=45
  swing1Motion=Free
  swing2LimitDegrees=45
  swing2Motion=Free
  twistLimitDegrees=45
  twistMotion=Free
 }
 {bone1Name=dyn_skirt_front_back_02_lf, bone2Name=dyn_skirt_front_back_01_lf, swing1LimitDegrees=7, swing2LimitDegrees=20, twistLimitDegrees=0}
 {bone1Name=dyn_skirt_front_back_03_lf, bone2Name=dyn_skirt_front_back_02_lf, swing1LimitDegrees=7, swing2LimitDegrees=20, twistLimitDegrees=0}
 {bone1Name=dyn_skirt_front_back_04_lf, bone2Name=dyn_skirt_front_back_03_lf, swing1LimitDegrees=7, swing2LimitDegrees=20, twistLimitDegrees=0}
 {bone1Name=dyn_skirt_front_02_lf, bone2Name=pelvis, swing1LimitDegrees=0, swing2LimitDegrees=20, twistLimitDegrees=0}
 {bone1Name=dyn_skirt_front_03_lf, bone2Name=dyn_skirt_front_02_lf, swing1LimitDegrees=10, swing2LimitDegrees=20, twistLimitDegrees=0}
 {bone1Name=dyn_skirt_front_04_lf, bone2Name=dyn_skirt_front_03_lf, swing1LimitDegrees=10, swing2LimitDegrees=20, twistLimitDegrees=10}
 {bone1Name=dyn_skirt_front_side_02_lf, bone2Name=pelvis, swing1LimitDegrees=0, swing2LimitDegrees=20, twistLimitDegrees=0}
 {bone1Name=dyn_skirt_front_side_03_lf, bone2Name=dyn_skirt_front_side_02_lf, swing1LimitDegrees=0, swing2LimitDegrees=20, twistLimitDegrees=0}
 {bone1Name=dyn_skirt_front_side_04_lf, bone2Name=dyn_skirt_front_side_03_lf, swing1LimitDegrees=0, swing2LimitDegrees=20, twistLimitDegrees=0}
 {
  bone1Name=pelvis
  bone2Name=root
  swing1LimitDegrees=45
  swing1Motion=Free
  swing2LimitDegrees=45
  swing2Motion=Free
  twistLimitDegrees=45
  twistMotion=Free
 }
 {
  bone1Name=offset_skirt_side_B_01_lf
  bone2Name=pelvis
  swing1LimitDegrees=45
  swing1Motion=Free
  swing2LimitDegrees=45
  swing2Motion=Free
  twistLimitDegrees=45
  twistMotion=Free
 }
 {
  bone1Name=dyn_skirt_side_B_01_lf
  bone2Name=offset_skirt_side_B_01_lf
  swing1LimitDegrees=45
  swing1Motion=Free
  swing2LimitDegrees=45
  swing2Motion=Free
  twistLimitDegrees=45
  twistMotion=Free
 }
 {
  bone1Name=dyn_skirt_side_B_02_lf
  bone2Name=dyn_skirt_side_B_01_lf
  swing1LimitDegrees=45
  swing1Motion=Free
  swing2LimitDegrees=45
  swing2Motion=Free
  twistLimitDegrees=45
  twistMotion=Free
 }
 {bone1Name=dyn_skirt_side_B_03_lf, bone2Name=dyn_skirt_side_B_02_lf, swing1LimitDegrees=0, swing2LimitDegrees=20, twistLimitDegrees=0}
 {bone1Name=dyn_skirt_side_B_04_lf, bone2Name=dyn_skirt_side_B_03_lf, swing1LimitDegrees=0, swing2LimitDegrees=20, twistLimitDegrees=0}
 {
  bone1Name=offset_skirt_side_A_01_lf
  bone2Name=pelvis
  swing1LimitDegrees=45
  swing1Motion=Free
  swing2LimitDegrees=45
  swing2Motion=Free
  twistLimitDegrees=45
  twistMotion=Free
 }
 {
  bone1Name=dyn_skirt_side_A_01_lf
  bone2Name=offset_skirt_side_A_01_lf
  swing1LimitDegrees=45
  swing1Motion=Free
  swing2LimitDegrees=45
  swing2Motion=Free
  twistLimitDegrees=45
  twistMotion=Free
 }
 {
  bone1Name=dyn_skirt_side_A_02_lf
  bone2Name=dyn_skirt_side_A_01_lf
  swing1LimitDegrees=45
  swing1Motion=Free
  swing2LimitDegrees=45
  swing2Motion=Free
  twistLimitDegrees=45
  twistMotion=Free
 }
 {bone1Name=dyn_skirt_side_A_03_lf, bone2Name=dyn_skirt_side_A_02_lf, swing1LimitDegrees=0, swing2LimitDegrees=20, twistLimitDegrees=0}
 {bone1Name=dyn_skirt_side_A_04_lf, bone2Name=dyn_skirt_side_A_03_lf, swing1LimitDegrees=0, swing2LimitDegrees=20, twistLimitDegrees=0}
]
