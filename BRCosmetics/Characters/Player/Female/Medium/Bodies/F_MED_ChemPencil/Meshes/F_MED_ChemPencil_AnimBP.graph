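# Post-process style AnimBP for the F_MED_ChemPencil female cosmetic body, adding skirt, belt and lace secondary motion to the pose copied from the attached parent.
# Cached poses chain additive bone offsets, a Control Rig and AnimDynamics laces into front-end or gameplay RigidBody simulation.
# A belt RigidBody, disabled while crouching, and final skirt offsets follow.
# EventGraph updates state flags and derives belt sim-space settings through a deprecated gravity-override function.
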AnimBlueprint /BRCosmetics/Characters/Player/Female/Medium/Bodies/F_MED_ChemPencil/Meshes/F_MED_ChemPencil_AnimBP
target_skeleton=&/F_MED_ChemPencil_Skeleton
graphs [
 AnimGraph {nodes [
   "Transform (Modify) Bone - Bone: dyn_skirt_side_A_01_lf"!(ComponentPose, bAlphaBoolEnabled:bool=True, Translation="0,0,0", Rotation="-0.195,-6.382,-4.586", Scale="1,1,1") -> (Pose) {
    alphaInputType=Bool
    rotationMode=BMM_Additive
    rotationSpace=BCS_BoneSpace
    alphaBoolBlend {blendInTime=0.2, blendOption=Linear, blendOutTime=0.2}
    boneToModify {boneName=dyn_skirt_side_A_01_lf}
    rotation=<rot -0.195 -6.3819 -4.586>
   }
   "Transform (Modify) Bone - Bone: dyn_skirt_side_A_01_lf"(bAlphaBoolEnabled=True, Translation="0,0,0", Rotation="7,5,0", Scale="1,1,1") {
    alphaInputType=Bool
    rotationMode=BMM_Additive
    rotationSpace=BCS_BoneSpace
    translationMode=BMM_Additive
    translationSpace=BCS_BoneSpace
    alphaBoolBlend {blendInTime=0.2, blendOption=Linear, blendOutTime=0.2}
    boneToModify {boneName=dyn_skirt_side_A_01_lf}
    rotation {pitch=7, yaw=5}
   }
   n1 = "Transform (Modify) Bone - Bone: dyn_skirt_front_back_02_lf"!(
    ComponentPose=Local To Component(LocalPose=Use cached pose 'RigidbodyPose'())
    bAlphaBoolEnabled:bool=True
    Translation="0,0,0"
    Rotation="10,0.008,0.004"
    Scale="1,1,1"
   ) -> (Pose) {
    alphaInputType=Bool
    rotationMode=BMM_Additive
    rotationSpace=BCS_BoneSpace
    alphaBoolBlend {blendInTime=0.1, blendOption=Linear, blendOutTime=0.1}
    boneToModify {boneName=dyn_skirt_front_back_02_lf}
    rotation=<rot 10 0.008 0.004>
   }
   n2 = "Transform (Modify) Bone - Bone: dyn_skirt_front_01_lf"!(ComponentPose=@n1, bAlphaBoolEnabled:bool=True, Translation="0,0,0", Rotation="-10.000,0.013,-0.004", Scale="1,1,1") -> (Pose) {
    alphaInputType=Bool
    rotationMode=BMM_Additive
    rotationSpace=BCS_BoneSpace
    alphaBoolBlend {blendInTime=0.1, blendOption=Linear, blendOutTime=0.1}
    boneToModify {boneName=dyn_skirt_front_01_lf}
    rotation=<rot -10 0.013 -0.004>
   }
   n3 = "Transform (Modify) Bone - Bone: offset_skirt_side_A_01_lf"!(ComponentPose=@n2, bAlphaBoolEnabled:bool=True, Translation="0,0,5", Rotation="0,0.013,0", Scale="1,1,1") -> (Pose) {
    alphaInputType=Bool
    rotationMode=BMM_Additive
    rotationSpace=BCS_BoneSpace
    translationMode=BMM_Additive
    translationSpace=BCS_BoneSpace
    alphaBoolBlend {blendInTime=0.1, blendOption=Linear, blendOutTime=0.1}
    boneToModify {boneName=offset_skirt_side_A_01_lf}
    rotation {yaw=0.013}
    translation {z=5}
   }
   n4 = "Transform (Modify) Bone - Bone: dyn_skirt_front_03_lf"!(ComponentPose=@n3, bAlphaBoolEnabled:bool=True, Translation="0,0,0", Rotation="10,0,0", Scale="1,1,1") -> (Pose) {
    alphaInputType=Bool
    rotationMode=BMM_Additive
    rotationSpace=BCS_BoneSpace
    alphaBoolBlend {blendInTime=0.2, blendOption=Linear, blendOutTime=0.2}
    boneToModify {boneName=dyn_skirt_front_03_lf}
    rotation {pitch=10}
   }
   n5 = "Transform (Modify) Bone - Bone: dyn_skirt_front_04_lf"!(ComponentPose=@n4, bAlphaBoolEnabled:bool=True, Translation="0,0,0", Rotation="-20,0,0", Scale="1,1,1") -> (Pose) {
    alphaInputType=Bool
    rotationMode=BMM_Additive
    rotationSpace=BCS_BoneSpace
    alphaBoolBlend {blendInTime=0.2, blendOption=Linear, blendOutTime=0.2}
    boneToModify {boneName=dyn_skirt_front_04_lf}
    rotation {pitch=-20}
   }
   n6 = "Transform (Modify) Bone - Bone: dyn_skirt_front_back_02_lf"(ComponentPose=@n5, bAlphaBoolEnabled=True, Translation="0,0,0", Rotation="-10,0,0", Scale="1,1,1") {
    alphaInputType=Bool
    rotationMode=BMM_Additive
    rotationSpace=BCS_BoneSpace
    alphaBoolBlend {blendInTime=0.2, blendOption=Linear, blendOutTime=0.2}
    boneToModify {boneName=dyn_skirt_front_back_02_lf}
    rotation {pitch=-10}
   }
   n7 = "Transform (Modify) Bone - Bone: dyn_skirt_back_02_lf"!(ComponentPose=@n6, bAlphaBoolEnabled:bool=True, Translation="0,0,0", Rotation="10,0,0", Scale="1,1,1") -> (Pose) {
    alphaInputType=Bool
    rotationMode=BMM_Additive
    rotationSpace=BCS_BoneSpace
    alphaBoolBlend {blendInTime=0.2, blendOption=Linear, blendOutTime=0.2}
    boneToModify {boneName=dyn_skirt_back_02_lf}
    rotation {pitch=10}
   }
   n8 = "Transform (Modify) Bone - Bone: dyn_skirt_back_03_lf"!(ComponentPose=@n7, bAlphaBoolEnabled:bool=True, Translation="0,0,0", Rotation="-10,0,0", Scale="1,1,1") -> (Pose) {
    alphaInputType=Bool
    rotationMode=BMM_Additive
    rotationSpace=BCS_BoneSpace
    alphaBoolBlend {blendInTime=0.2, blendOption=Linear, blendOutTime=0.2}
    boneToModify {boneName=dyn_skirt_back_03_lf}
    rotation {pitch=-10}
   }
   n9 = "Transform (Modify) Bone - Bone: dyn_skirt_side_A_01_lf"(ComponentPose=@n8, bAlphaBoolEnabled=True, Translation="0,0,0", Rotation="-2,0,0", Scale="1,1,1") {
    alphaInputType=Bool
    rotationMode=BMM_Additive
    rotationSpace=BCS_BoneSpace
    alphaBoolBlend {blendInTime=0.2, blendOption=Linear, blendOutTime=0.2}
    boneToModify {boneName=dyn_skirt_side_A_01_lf}
    rotation {pitch=-2}
   }
   n10 = "Transform (Modify) Bone - Bone: dyn_skirt_side_A_02_lf"!(ComponentPose=@n9, bAlphaBoolEnabled:bool=True, Translation="0,0,0", Rotation="-2,0,0", Scale="1,1,1") -> (Pose) {
    alphaInputType=Bool
    rotationMode=BMM_Additive
    rotationSpace=BCS_BoneSpace
    alphaBoolBlend {blendInTime=0.2, blendOption=Linear, blendOutTime=0.2}
    boneToModify {boneName=dyn_skirt_side_A_02_lf}
    rotation {pitch=-2}
   }
   n11 = "Transform (Modify) Bone - Bone: dyn_skirt_side_A_03_lf"!(ComponentPose=@n10, bAlphaBoolEnabled:bool=True, Translation="0,0,0", Rotation="10,0,0", Scale="1,1,1") -> (Pose) {
    alphaInputType=Bool
    rotationMode=BMM_Additive
    rotationSpace=BCS_BoneSpace
    alphaBoolBlend {blendInTime=0.2, blendOption=Linear, blendOutTime=0.2}
    boneToModify {boneName=dyn_skirt_side_A_03_lf}
    rotation {pitch=10}
   }
   n12 = "Transform (Modify) Bone - Bone: dyn_skirt_side_A_04_lf"!(ComponentPose=@n11, bAlphaBoolEnabled:bool=True, Translation="0,0,0", Rotation="0,0,0", Scale="1,1,1") -> (Pose) {
    alphaInputType=Bool
    rotationMode=BMM_Additive
    rotationSpace=BCS_BoneSpace
    alphaBoolBlend {blendInTime=0.2, blendOption=Linear, blendOutTime=0.2}
    boneToModify {boneName=dyn_skirt_side_A_04_lf}
   }
   Output Pose(Result=Component To Local(ComponentPose=@n12)) {layerGroup=None}
   n13 = Control Rig!(
    Source=Use cached pose 'PreFixPose'()
    Alpha
    Crouch:bool=Get bIsCrouching()
    Moving:double=0
    emote:bool=Get bIsPlayingEmote()
    RBAN:bool=NOT Boolean(A=Get bRigidBodyEnabled())
    frontend:bool=Get bIsInFrontEnd()
   ) -> (Pose) {
    alphaBoolBlend {blendOption=Linear}
    controlRigAssetReference {blueprintRigClass=&/F_MED_ChemPencil_CtrlRig.F_MED_ChemPencil_CtrlRig_C}
   }
   n14 = "Transform (Modify) Bone - Bone: dyn_belt_front_01_lf"!(ComponentPose=Local To Component(LocalPose=@n13), bAlphaBoolEnabled:bool=True, Translation="0,0,0", Rotation="15,0,0", Scale="1,1,1") -> (Pose) {
    alphaInputType=Bool
    rotationMode=BMM_Additive
    rotationSpace=BCS_BoneSpace
    translationMode=BMM_Additive
    translationSpace=BCS_BoneSpace
    alphaBoolBlend {blendInTime=0.2, blendOption=Linear, blendOutTime=0.2}
    boneToModify {boneName=dyn_belt_front_01_lf}
    rotation {pitch=15}
   }
   n15 = "Transform (Modify) Bone - Bone: dyn_skirt_side_A_01_lf"(ComponentPose=@n14, bAlphaBoolEnabled=True, Translation="0,0,0", Rotation="5,5,0", Scale="1,1,1") {
    alphaInputType=Bool
    rotationMode=BMM_Additive
    rotationSpace=BCS_BoneSpace
    translationMode=BMM_Additive
    translationSpace=BCS_BoneSpace
    alphaBoolBlend {blendInTime=0.2, blendOption=Linear, blendOutTime=0.2}
    boneToModify {boneName=dyn_skirt_side_A_01_lf}
    rotation {pitch=5, yaw=5}
   }
   n16 = "Transform (Modify) Bone - Bone: offset_skirt_front_lf"!(ComponentPose=@n15, bAlphaBoolEnabled:bool=True, Translation="0,0,0", Rotation="5,0,0", Scale="1,1,1") -> (Pose) {
    alphaInputType=Bool
    rotationMode=BMM_Additive
    rotationSpace=BCS_BoneSpace
    translationMode=BMM_Additive
    translationSpace=BCS_BoneSpace
    alphaBoolBlend {blendInTime=0.2, blendOption=Linear, blendOutTime=0.2}
    boneToModify {boneName=offset_skirt_front_lf}
    rotation {pitch=5}
   }
   n17 = "Transform (Modify) Bone - Bone: offset_skirt_front_lf"(ComponentPose=@n16, bAlphaBoolEnabled=True, Translation="0,0,0", Rotation="5,0,0", Scale="1,1,1") {
    alphaInputType=Bool
    rotationMode=BMM_Additive
    rotationSpace=BCS_BoneSpace
    translationMode=BMM_Additive
    translationSpace=BCS_BoneSpace
    alphaBoolBlend {blendInTime=0.2, blendOption=Linear, blendOutTime=0.2}
    boneToModify {boneName=offset_skirt_front_lf}
    rotation {pitch=5}
   }
   n18 = "Transform (Modify) Bone - Bone: dyn_belt_front_01_lf"(ComponentPose=@n17, bAlphaBoolEnabled=True, Translation="0,0,0", Rotation="10,0,0", Scale="1,1,1") {
    alphaInputType=Bool
    rotationMode=BMM_Additive
    rotationSpace=BCS_BoneSpace
    translationMode=BMM_Additive
    translationSpace=BCS_BoneSpace
    alphaBoolBlend {blendInTime=0.2, blendOption=Linear, blendOutTime=0.2}
    boneToModify {boneName=dyn_belt_front_01_lf}
    rotation {pitch=10}
   }
   n19 = "Transform (Modify) Bone - Bone: dyn_skirt_front_01_lf"(ComponentPose=@n18, bAlphaBoolEnabled=True, Translation="0,0,0", Rotation="-20,0,0", Scale="1,1,1") {
    alphaInputType=Bool
    rotationMode=BMM_Additive
    rotationSpace=BCS_BoneSpace
    translationMode=BMM_Additive
    translationSpace=BCS_BoneSpace
    alphaBoolBlend {blendInTime=0.2, blendOption=Linear, blendOutTime=0.2}
    boneToModify {boneName=dyn_skirt_front_01_lf}
    rotation {pitch=-20}
   }
   n20 = "Transform (Modify) Bone - Bone: dyn_skirt_front_side_01_lf"!(ComponentPose=@n19, bAlphaBoolEnabled:bool=True, Translation="0,0,0", Rotation="-20,0,0", Scale="1,1,1") -> (Pose) {
    alphaInputType=Bool
    rotationMode=BMM_Additive
    rotationSpace=BCS_BoneSpace
    translationMode=BMM_Additive
    translationSpace=BCS_BoneSpace
    alphaBoolBlend {blendInTime=0.2, blendOption=Linear, blendOutTime=0.2}
    boneToModify {boneName=dyn_skirt_front_side_01_lf}
    rotation {pitch=-20}
   }
   n21 = "Transform (Modify) Bone - Bone: offset_skirt_front_back_lf"!(ComponentPose=@n20, bAlphaBoolEnabled:bool=True, Translation="0,0,0", Rotation="90,0,0", Scale="1,1,1") -> (Pose) {
    alphaInputType=Bool
    rotationMode=BMM_Additive
    rotationSpace=BCS_BoneSpace
    translationMode=BMM_Additive
    translationSpace=BCS_BoneSpace
    alphaBoolBlend {blendInTime=0.2, blendOption=Linear, blendOutTime=0.2}
    boneToModify {boneName=offset_skirt_front_back_lf}
    rotation {pitch=90}
   }
   n22 = "Transform (Modify) Bone - Bone: offset_skirt_back_lf"!(ComponentPose=@n21, bAlphaBoolEnabled:bool=True, Translation="0,0,0", Rotation="90,0,0", Scale="1,1,1") -> (Pose) {
    alphaInputType=Bool
    rotationMode=BMM_Additive
    rotationSpace=BCS_BoneSpace
    translationMode=BMM_Additive
    translationSpace=BCS_BoneSpace
    alphaBoolBlend {blendInTime=0.2, blendOption=Linear, blendOutTime=0.2}
    boneToModify {boneName=offset_skirt_back_lf}
    rotation {pitch=90}
   }
   n23 = "Transform (Modify) Bone - Bone: offset_skirt_side_B_01_lf"!(ComponentPose=@n22, bAlphaBoolEnabled:bool=True, Translation="0,0,0", Rotation="19.373,-64.651,-18.307", Scale="1,1,1") -> (Pose) {
    alphaInputType=Bool
    rotationMode=BMM_Additive
    rotationSpace=BCS_BoneSpace
    translationMode=BMM_Additive
    translationSpace=BCS_BoneSpace
    alphaBoolBlend {blendInTime=0.2, blendOption=Linear, blendOutTime=0.2}
    boneToModify {boneName=offset_skirt_side_B_01_lf}
    rotation=<rot 19.373 -64.651 -18.307>
   }
   n24 = "Transform (Modify) Bone - Bone: offset_skirt_front_lf"(ComponentPose=@n23, bAlphaBoolEnabled=True, Translation="0,0,0", Rotation="10,0,0", Scale="1,1,1") {
    alphaInputType=Bool
    rotationMode=BMM_Additive
    rotationSpace=BCS_BoneSpace
    alphaBoolBlend {blendInTime=0.2, blendOption=Linear, blendOutTime=0.2}
    boneToModify {boneName=offset_skirt_front_lf}
    rotation {pitch=10}
   }
   Save cached pose 'ConstraintsPose'(Pose=Component To Local(ComponentPose=@n24)) {cacheName=ConstraintsPose}
   n25 = Anim Dynamics - Bone: dyn_laces_01_lf(ComponentPose=Local To Component(LocalPose=Use cached pose 'ConstraintsPose'()), Alpha=1, ExternalForce="0,0,0") {
    angularDampingOverride=1
    angularSpringConstant=20000
    bAngularSpring=true
    bOverrideAngularDamping=true
    bUseGravityOverride=true
    gravityScale=0.35
    simulationSpace=RootRelative
    alphaBoolBlend {blendOption=Linear}
    boundBone {boneName=dyn_laces_01_lf}
    componentLinearAccScale=<1 1 1>
    componentLinearVelScale=<1 1 1>
    gravityOverride {z=-980}
    physicsBodyDefinitions [{
      boundBone {boneName=dyn_laces_01_lf}
      boxExtents=<20 30 30>
      constraintSetup {angularLimitsMax=<0 5 0>, angularLimitsMin=<0 -15 0>, angularTarget=<1 0 0>}
      localJointOffset=<4 0 0>
     }]
   }
   n26 = Anim Dynamics - Bone: dyn_laces_02_lf(ComponentPose=@n25, Alpha=1, ExternalForce="0,0,0") {
    angularDampingOverride=1
    angularSpringConstant=20000
    bAngularSpring=true
    bOverrideAngularDamping=true
    bUseGravityOverride=true
    gravityScale=0.35
    simulationSpace=RootRelative
    alphaBoolBlend {blendOption=Linear}
    boundBone {boneName=dyn_laces_02_lf}
    componentLinearAccScale=<1 1 1>
    componentLinearVelScale=<1 1 1>
    gravityOverride {z=-980}
    physicsBodyDefinitions [{
      boundBone {boneName=dyn_laces_02_lf}
      boxExtents=<20 30 30>
      constraintSetup {angularLimitsMax=<0 5 0>, angularLimitsMin=<0 -15 0>, angularTarget=<1 0 0>}
      localJointOffset=<4 0 0>
     }]
   }
   n27 = Anim Dynamics - Bone: dyn_laces_01_rt(ComponentPose=@n26, Alpha=1, ExternalForce="0,0,0") {
    angularDampingOverride=1
    angularSpringConstant=20000
    bAngularSpring=true
    bOverrideAngularDamping=true
    bUseGravityOverride=true
    gravityScale=0.35
    simulationSpace=RootRelative
    alphaBoolBlend {blendOption=Linear}
    boundBone {boneName=dyn_laces_01_rt}
    componentLinearAccScale=<1 1 1>
    componentLinearVelScale=<1 1 1>
    gravityOverride {z=-980}
    physicsBodyDefinitions [{
      boundBone {boneName=dyn_laces_01_rt}
      boxExtents=<20 30 30>
      constraintSetup {angularLimitsMax=<0 5 0>, angularLimitsMin=<0 -15 0>, angularTarget=<1 0 0>}
      localJointOffset=<4 0 0>
     }]
   }
   n28 = Anim Dynamics - Bone: dyn_laces_02_rt(ComponentPose=@n27, Alpha=1, ExternalForce="0,0,0") {
    angularDampingOverride=1
    angularSpringConstant=20000
    bAngularSpring=true
    bOverrideAngularDamping=true
    bUseGravityOverride=true
    gravityScale=0.35
    simulationSpace=RootRelative
    alphaBoolBlend {blendOption=Linear}
    boundBone {boneName=dyn_laces_02_rt}
    componentLinearAccScale=<1 1 1>
    componentLinearVelScale=<1 1 1>
    gravityOverride {z=-980}
    physicsBodyDefinitions [{
      boundBone {boneName=dyn_laces_02_rt}
      boxExtents=<20 30 30>
      constraintSetup {angularLimitsMax=<0 5 0>, angularLimitsMin=<0 -15 0>, angularTarget=<1 0 0>}
      localJointOffset=<4 0 0>
     }]
   }
   Save cached pose 'LacesPose'(Pose=Component To Local(ComponentPose=@n28)) {cacheName=LacesPose}
   n29 = "Transform (Modify) Bone - Bone: dyn_skirt_front_02_lf"!(
    ComponentPose=Local To Component(LocalPose=Copy Pose From Mesh() {bUseAttachedParent=true})
    bAlphaBoolEnabled:bool=True
    Translation="0,0,0"
    Rotation="0,10,0"
    Scale="1,1,1"
   ) -> (Pose) {
    alphaInputType=Bool
    rotationMode=BMM_Additive
    rotationSpace=BCS_BoneSpace
    alphaBoolBlend {blendInTime=0.1, blendOption=Linear, blendOutTime=0.1}
    boneToModify {boneName=dyn_skirt_front_02_lf}
    rotation {yaw=10}
   }
   n30 = "Transform (Modify) Bone - Bone: dyn_skirt_front_01_01_lf"!(ComponentPose=@n29, bAlphaBoolEnabled:bool=True, Translation="0,0,0", Rotation="10,-20,0", Scale="1,1,1") -> (Pose) {
    alphaInputType=Bool
    rotationMode=BMM_Additive
    rotationSpace=BCS_BoneSpace
    alphaBoolBlend {blendInTime=0.1, blendOption=Linear, blendOutTime=0.1}
    boneToModify {boneName=dyn_skirt_front_01_01_lf}
    rotation {pitch=10, yaw=-20}
   }
   n31 = "Transform (Modify) Bone - Bone: dyn_skirt_front_01_lf"(ComponentPose=@n30, bAlphaBoolEnabled=True, Translation="0,0,0", Rotation="-12,0,0", Scale="1,1,1") {
    alphaInputType=Bool
    rotationMode=BMM_Additive
    rotationSpace=BCS_BoneSpace
    alphaBoolBlend {blendInTime=0.1, blendOption=Linear, blendOutTime=0.1}
    boneToModify {boneName=dyn_skirt_front_01_lf}
    rotation {pitch=-12}
   }
   n32 = "Transform (Modify) Bone - Bone: dyn_skirt_front_side_01_lf"(ComponentPose=@n31, bAlphaBoolEnabled=True, Translation="0,0,0", Rotation="0,0,0", Scale="1,1,1") {
    alphaInputType=Bool
    rotationMode=BMM_Additive
    rotationSpace=BCS_BoneSpace
    alphaBoolBlend {blendInTime=0.1, blendOption=Linear, blendOutTime=0.1}
    boneToModify {boneName=dyn_skirt_front_side_01_lf}
   }
   n33 = "Transform (Modify) Bone - Bone: dyn_skirt_side_A_01_lf"(ComponentPose=@n32, bAlphaBoolEnabled=True, Translation="0,0,0", Rotation="0,5,0", Scale="1,1,1") {
    alphaInputType=Bool
    rotationMode=BMM_Additive
    rotationSpace=BCS_BoneSpace
    alphaBoolBlend {blendInTime=0.2, blendOption=Linear, blendOutTime=0.2}
    boneToModify {boneName=dyn_skirt_side_A_01_lf}
    rotation {yaw=5}
   }
   n34 = "Transform (Modify) Bone - Bone: dyn_skirt_side_B_01_lf"!(ComponentPose=@n33, bAlphaBoolEnabled:bool=True, Translation="0,0,0", Rotation="-3,0,0", Scale="1,1,1") -> (Pose) {
    alphaInputType=Bool
    rotationMode=BMM_Additive
    rotationSpace=BCS_BoneSpace
    alphaBoolBlend {blendInTime=0.2, blendOption=Linear, blendOutTime=0.2}
    boneToModify {boneName=dyn_skirt_side_B_01_lf}
    rotation {pitch=-3}
   }
   n35 = "Transform (Modify) Bone - Bone: dyn_skirt_side_B_01_lf"(ComponentPose=@n34, bAlphaBoolEnabled=True, Translation="0,0,0", Rotation="4,-5,0", Scale="1,1,1") {
    alphaInputType=Bool
    rotationMode=BMM_Additive
    rotationSpace=BCS_BoneSpace
    alphaBoolBlend {blendInTime=0.2, blendOption=Linear, blendOutTime=0.2}
    boneToModify {boneName=dyn_skirt_side_B_01_lf}
    rotation {pitch=4, yaw=-5}
   }
   n36 = "Transform (Modify) Bone - Bone: dyn_skirt_side_B_02_lf"!(ComponentPose=@n35, bAlphaBoolEnabled:bool=True, Translation="0,0,0", Rotation="0,5,0", Scale="1,1,1") -> (Pose) {
    alphaInputType=Bool
    rotationMode=BMM_Additive
    rotationSpace=BCS_BoneSpace
    alphaBoolBlend {blendInTime=0.2, blendOption=Linear, blendOutTime=0.2}
    boneToModify {boneName=dyn_skirt_side_B_02_lf}
    rotation {yaw=5}
   }
   n37 = "Transform (Modify) Bone - Bone: dyn_skirt_side_A_01_lf"(ComponentPose=@n36, bAlphaBoolEnabled=True, Translation="0,0,0", Rotation="-2,0,0", Scale="1,1,1") {
    alphaInputType=Bool
    rotationMode=BMM_Additive
    rotationSpace=BCS_BoneSpace
    alphaBoolBlend {blendInTime=0.2, blendOption=Linear, blendOutTime=0.2}
    boneToModify {boneName=dyn_skirt_side_A_01_lf}
    rotation {pitch=-2}
   }
   Save cached pose 'PreFixPose'(Pose=Component To Local(ComponentPose=@n37)) {cacheName=PreFixPose}
   n38 = "Transform (Modify) Bone - Bone: dyn_belt_back_01_lf"!(
    ComponentPose=Local To Component(LocalPose=Use cached pose 'LacesPose'())
    bAlphaBoolEnabled:bool=True
    Translation="0,0,0"
    Rotation="10.004,-9.999,0.006"
    Scale="1,1,1"
   ) -> (Pose) {
    alphaInputType=Bool
    rotationMode=BMM_Additive
    rotationSpace=BCS_BoneSpace
    alphaBoolBlend {blendInTime=0.2, blendOption=Linear, blendOutTime=0.2}
    boneToModify {boneName=dyn_belt_back_01_lf}
    rotation=<rot 10.004 -9.999 0.006>
   }
   n39 = "Transform (Modify) Bone - Bone: dyn_skirt_front_02_lf"(ComponentPose=@n38, bAlphaBoolEnabled=True, Translation="0,0,0", Rotation="-10,0,0.006", Scale="1,1,1") {
    alphaInputType=Bool
    rotationMode=BMM_Additive
    rotationSpace=BCS_BoneSpace
    alphaBoolBlend {blendInTime=0.2, blendOption=Linear, blendOutTime=0.2}
    boneToModify {boneName=dyn_skirt_front_02_lf}
    rotation {pitch=-10, roll=0.006}
   }
   n40 = "Transform (Modify) Bone - Bone: dyn_skirt_front_side_02_lf"!(ComponentPose=@n39, bAlphaBoolEnabled:bool=True, Translation="0,0,0", Rotation="-15,0,0.006", Scale="1,1,1") -> (Pose) {
    alphaInputType=Bool
    rotationMode=BMM_Additive
    rotationSpace=BCS_BoneSpace
    alphaBoolBlend {blendInTime=0.2, blendOption=Linear, blendOutTime=0.2}
    boneToModify {boneName=dyn_skirt_front_side_02_lf}
    rotation {pitch=-15, roll=0.006}
   }
   n41 = "Transform (Modify) Bone - Bone: dyn_skirt_side_A_01_lf"(ComponentPose=@n40, bAlphaBoolEnabled=True, Translation="0,0,0", Rotation="0,-10,0.006", Scale="1,1,1") {
    alphaInputType=Bool
    rotationMode=BMM_Additive
    rotationSpace=BCS_BoneSpace
    alphaBoolBlend {blendInTime=0.2, blendOption=Linear, blendOutTime=0.2}
    boneToModify {boneName=dyn_skirt_side_A_01_lf}
    rotation {roll=0.006, yaw=-10}
   }
   n42 = "Transform (Modify) Bone - Bone: base_belt_front_lf"!(ComponentPose=@n41, bAlphaBoolEnabled:bool=True, Translation="0,0,0", Rotation="10,0.02,0.027", Scale="1,1,1") -> (Pose) {
    alphaInputType=Bool
    rotationMode=BMM_Additive
    rotationSpace=BCS_BoneSpace
    alphaBoolBlend {blendInTime=0.2, blendOption=Linear, blendOutTime=0.2}
    boneToModify {boneName=base_belt_front_lf}
    rotation=<rot 10 0.02 0.027>
   }
   n43 = "Transform (Modify) Bone - Bone: offset_skirt_front_side_lf"!(ComponentPose=@n42, bAlphaBoolEnabled:bool=True, Translation="0,0,0", Rotation="9.996,0.020,0.027", Scale="1,1,1") -> (Pose) {
    alphaInputType=Bool
    rotationMode=BMM_Additive
    rotationSpace=BCS_BoneSpace
    alphaBoolBlend {blendInTime=0.2, blendOption=Linear, blendOutTime=0.2}
    boneToModify {boneName=offset_skirt_front_side_lf}
    rotation=<rot 9.9961 0.02 0.027>
   }
   n44 = "Transform (Modify) Bone - Bone: dyn_skirt_front_01_lf"(ComponentPose=@n43, bAlphaBoolEnabled=True, Translation="0,0,0", Rotation="6.622,14.555,-3.175", Scale="1,1,1") {
    alphaInputType=Bool
    rotationMode=BMM_Additive
    rotationSpace=BCS_BoneSpace
    alphaBoolBlend {blendInTime=0.2, blendOption=Linear, blendOutTime=0.2}
    boneToModify {boneName=dyn_skirt_front_01_lf}
    rotation=<rot 6.6221 14.555 -3.175>
   }
   n45 = "Transform (Modify) Bone - Bone: offset_skirt_side_A_01_lf"(ComponentPose=@n44, bAlphaBoolEnabled=True, Translation="0,0,0", Rotation="7,0,0", Scale="1,1,1") {
    alphaInputType=Bool
    rotationMode=BMM_Additive
    rotationSpace=BCS_BoneSpace
    alphaBoolBlend {blendInTime=0.1, blendOption=Linear, blendOutTime=0.1}
    boneToModify {boneName=offset_skirt_side_A_01_lf}
    rotation {pitch=7}
   }
   n46 = "Transform (Modify) Bone - Bone: dyn_skirt_front_01_lf"(ComponentPose=@n45, bAlphaBoolEnabled=True, Translation="0,0,0", Rotation="10,0,0", Scale="1,1,1") {
    alphaInputType=Bool
    rotationMode=BMM_Additive
    rotationSpace=BCS_BoneSpace
    alphaBoolBlend {blendInTime=0.3, blendOption=Linear, blendOutTime=0.3}
    boneToModify {boneName=dyn_skirt_front_01_lf}
    rotation {pitch=10}
   }
   Save cached pose 'PreLogicPose'(Pose=Component To Local(ComponentPose=@n46)) {cacheName=PreLogicPose}
   n47 = NOT Boolean(A=Get bIsPlayingEmote())
   n48 = "Transform (Modify) Bone - Bone: dyn_skirt_back_01_lf"!(
    ComponentPose=Local To Component(LocalPose=Use cached pose 'PreLogicPose'())
    bAlphaBoolEnabled:bool=@n47
    Translation="0,0,0"
    Rotation="-5,0,0"
    Scale="1,1,1"
   ) -> (Pose) {
    alphaInputType=Bool
    rotationMode=BMM_Additive
    rotationSpace=BCS_BoneSpace
    alphaBoolBlend {blendOption=Linear}
    boneToModify {boneName=dyn_skirt_back_01_lf}
    rotation {pitch=-5}
   }
   n49 = "Transform (Modify) Bone - Bone: dyn_skirt_back_02_lf"(ComponentPose=@n48, bAlphaBoolEnabled=@n47, Translation="0,0,0", Rotation="-5,0,0", Scale="1,1,1") {
    alphaInputType=Bool
    rotationMode=BMM_Additive
    rotationSpace=BCS_BoneSpace
    alphaBoolBlend {blendOption=Linear}
    boneToModify {boneName=dyn_skirt_back_02_lf}
    rotation {pitch=-5}
   }
   n50 = RigidBody!(ComponentPose=@n49, Alpha, OverrideWorldGravity:Vector="0,0,0", ExternalForce) -> (Pose) {
    bClampLinearTranslationLimitToRefPose=true
    bOverrideWorldGravity=true
    overridePhysicsAsset=&/F_MED_ChemPencil_FrontEnd_Physics
    simulationSpace=BaseBoneSpace
    alphaBoolBlend {blendOption=Linear}
    baseBoneRef {boneName=pelvis}
    simSpaceSettings {maxAngularAcceleration=200, maxAngularVelocity=220, maxLinearAcceleration=200, maxLinearVelocity=200}
   }
   n51 = Component To Local(ComponentPose=@n50)
   n52 = RigidBody(ComponentPose=Local To Component(LocalPose=Use cached pose 'PreLogicPose'()), OverrideWorldGravity="0,0,0") {
    bClampLinearTranslationLimitToRefPose=true
    bOverrideWorldGravity=true
    overridePhysicsAsset=&/F_MED_ChemPencil_Physics
    simulationSpace=BaseBoneSpace
    alphaBoolBlend {blendOption=Linear}
    baseBoneRef {boneName=pelvis}
    simSpaceSettings {worldAlpha=0.05, externalAngularVelocity=<1 1>, externalLinearDragV {x=1}, externalLinearVelocity=<1 1>}
   }
   n53 = Component To Local(ComponentPose=@n52)
   n54 = Get bIsInFrontEnd()
   n55 = RigidBody!(
    ComponentPose=Local To Component(LocalPose=Blend Poses by bool(BlendPose_0=@n51, BlendPose_1=@n53, bActiveValue=@n54) {blendType=Linear})
    bAlphaBoolEnabled:bool=NOT Boolean(A=Get bIsCrouching())
    OverrideWorldGravity="0,0,0"
    ExternalForce
    SimSpaceSettings:Sim Space Settings Structure=Get Sim Space Settings() -> (Sim Space Settings:Sim Space Settings Structure)
   ) -> (Pose) {
    alphaInputType=Bool
    bClampLinearTranslationLimitToRefPose=true
    bOverrideWorldGravity=true
    overridePhysicsAsset=&/F_MED_ChemPencil_Belt_Physics
    simulationSpace=BaseBoneSpace
    alphaBoolBlend {blendInTime=0.2, blendOption=Linear, blendOutTime=0.2}
    baseBoneRef {boneName=pelvis}
    simSpaceSettings {worldAlpha=0.3, externalAngularVelocity {x=5}, externalLinearVelocity {x=10}}
   }
   Save cached pose 'RigidbodyPose'(Pose=Component To Local(ComponentPose=@n55)) {cacheName=RigidbodyPose}
  ]}
 EventGraph {nodes [
   Event BlueprintUpdateAnimation()
   > n1 = Sequence!(execute) -> (then_0, then_1, then_2:Exec, then_3:Exec)
   n2 = Get bIsJumping()
   n3 = Get bIsFalling()
   Set Jump(execute=@n1.then_1, Jump:bool=OR Boolean(A=@n2, B=@n3)) -> (Output_Get:bool)
   n4 = NOR Boolean!(A:bool=Get bIsMoving2D(), B:bool=Get bIsSkyDiving()) -> (ReturnValue:bool)
   n5 = Get bIsPlayingEmote()
   > Set IDLE(IDLE:bool=NOR Boolean(A=NOR Boolean(A=@n4, B=@n5), B=Get Jump() -> (Jump:bool))) -> (Output_Get:bool)
   > Set WalkinCrouching(WalkinCrouching:bool=AND Boolean(A=Get bIsMoving2D(), B=Get bIsCrouching())) -> (Output_Get:bool)
   n6 = Get bIsCrouching()
   > Set SprinitingNotCrouching(SprinitingNotCrouching:bool=AND Boolean(A=Get bIsMoving2D(), B=NOT Boolean(A=@n6))) -> (Output_Get:bool)
   n7 = Get bIsPlayingEmote()
   Set FrontIDLE(execute=@n1.then_2, FrontIDLE:bool=AND Boolean(A=Get bIsInFrontEnd(), B=NOT Boolean(A=@n7))) -> (Output_Get:bool)
   n8 = Get bIsMoving2D()
   > Set CrouchIDLE(CrouchIDLE:bool=AND Boolean(A=Get bIsCrouching(), B=NOT Boolean(A=@n8))) -> (Output_Get:bool)
   n9 = Sequence!(execute=@n1.then_3) -> (then_0, then_1)
   n10 = GetAnimInstance(self=GetOwnerSkeletalMeshComponent())
   Is Valid(exec=@n9.then_1, InputObject=@n10)
   > Set isBalletSpin(isBalletSpin:bool=Montage_IsPlaying(self=@n10, Montage=/BRCosmetics/Animation/Game/MainPlayer/Emotes/BalletSpin/Emote_BalletSpin_CMF_M)) -> (Output_Get:bool)
   n11 = GetAnimInstance(self=GetOwnerSkeletalMeshComponent())
   Is Valid(exec=@n9.then_0, InputObject=@n11)
   > Set isCustomEmotePlaying(isCustomEmotePlaying:bool=Montage_IsPlaying(self=@n11, Montage=/BRCosmetics/Animation/Game/MainPlayer/Emotes/SunBurst/Emote_SunBurst_CMF_Skirt_M)) -> (Output_Get:bool)
   n12 = Gravity Override State Function DEPRECATED(
    execute=@n1.then_0
    CharacterPartAnimInstance=Self-Reference()
    GameTimeInSeconds=GetGameTimeInSeconds()
    GravityOverrideParams=Get Gravity Override Params()
    RBAN Params=Get RBAN Params()
   )
   > Set GravityOverride(GravityOverride=@n12)
   > Set Sim Space Settings(Sim Space Settings:Sim Space Settings Structure=Break RBANSetting Struct!(RBANSettingStruct:"RBANSetting Struct Structure (by ref)"=@n12.RBAN SettingsOut) -> (
     ComponentLinearAccScale_28_#1:Vector=~
     ComponentLinearVelScale_31_#2:Vector=~
     SimSpaceSettings_34_#3:Sim Space Settings Structure
    ) {is=title K2Node_BreakStruct}.SimSpaceSettings_34_#3) -> (Output_Get:Sim Space Settings Structure)
   AND Boolean(A=@n2, B=@n3)
  ]}
]
variables [
 CrouchIDLE {type=boolean}
 FrontIDLE {type=boolean}
 Gravity Override Params {type=GravityOverrideParamsStruct}
 GravityOverride {type=Vector}
 IDLE {type=boolean}
 isBalletSpin {type=boolean}
 isCustomEmotePlaying {type=boolean}
 Jump {type=boolean}
 Jumping {type=boolean}
 RBAN Params {
  type=RBANParamsStruct
  default {
   dBNO {simSpaceSettings {
     maxAngularAcceleration=1000
     maxAngularVelocity=3
     maxLinearAcceleration=300
     maxLinearVelocity=700
     velocityScaleZ=0
     externalLinearDragV=<0 0.5 0>
    }}
   dBNO_Carried {simSpaceSettings {
     maxAngularAcceleration=1000
     maxAngularVelocity=3
     maxLinearAcceleration=300
     maxLinearVelocity=700
     velocityScaleZ=0
     externalLinearDragV=<0 0.5 0>
    }}
   emoteOrMelee {simSpaceSettings {
     maxAngularAcceleration=1000
     maxAngularVelocity=3
     maxLinearAcceleration=300
     maxLinearVelocity=700
     velocityScaleZ=0
     externalLinearDragV=<0 0.5 0>
    }}
   front_End {simSpaceSettings {
     maxAngularAcceleration=1000
     maxAngularVelocity=3
     maxLinearAcceleration=300
     maxLinearVelocity=200
     velocityScaleZ=0
     worldAlpha=0
     externalLinearDragV=<0 0 0>
    }}
   onGround_Crouch_Moving {simSpaceSettings {
     maxAngularAcceleration=1000
     maxAngularVelocity=3
     maxLinearAcceleration=300
     maxLinearVelocity=700
     velocityScaleZ=0
     externalLinearDragV=<0 0.5 0>
    }}
   onGround_Falling {simSpaceSettings {
     maxAngularAcceleration=1000
     maxAngularVelocity=3
     maxLinearAcceleration=300
     maxLinearVelocity=700
     velocityScaleZ=0
     externalLinearDragV=<0 0.5 0>
    }}
   onGround_Moving {simSpaceSettings {
     maxAngularAcceleration=1000
     maxAngularVelocity=3
     maxLinearAcceleration=300
     maxLinearVelocity=700
     velocityScaleZ=0
     externalAngularVelocity=<0 5 0>
     externalLinearDragV=<0 0.5 0>
     externalLinearVelocity=<0 5 0>
    }}
   onGround_Standing {simSpaceSettings {
     maxAngularAcceleration=1000
     maxAngularVelocity=3
     maxLinearAcceleration=300
     maxLinearVelocity=700
     velocityScaleZ=0
     externalLinearDragV=<0 0.5 0>
    }}
   rail_Grinding {simSpaceSettings {
     maxAngularAcceleration=1000
     maxAngularVelocity=3
     maxLinearAcceleration=300
     maxLinearVelocity=700
     velocityScaleZ=0
     worldAlpha=0.1
     externalLinearDragV=<0 0.5 0>
    }}
   riding_Moving {simSpaceSettings {
     maxAngularAcceleration=1000
     maxAngularVelocity=3
     maxLinearAcceleration=300
     maxLinearVelocity=700
     velocityScaleZ=0
     externalLinearDragV=<0 0.5 0>
    }}
   riding_Moving_Targeting {simSpaceSettings {
     maxAngularAcceleration=1000
     maxAngularVelocity=3
     maxLinearAcceleration=300
     maxLinearVelocity=700
     velocityScaleZ=0
     externalLinearDragV=<0 0.5 0>
    }}
   riding_Standing {simSpaceSettings {
     maxAngularAcceleration=1000
     maxAngularVelocity=3
     maxLinearAcceleration=300
     maxLinearVelocity=700
     velocityScaleZ=0
     externalLinearDragV=<0 0.5 0>
    }}
   skydive_DivingDown {simSpaceSettings {
     maxAngularAcceleration=1000
     maxAngularVelocity=3
     maxLinearAcceleration=300
     maxLinearVelocity=700
     velocityScaleZ=0
     externalLinearDragV=<0 0.5 0>
    }}
   skydive_FloatingDown {
    componentLinearAccScale=<0 0.5 0>
    componentLinearVelScale=<0 0.5 0>
    simSpaceSettings {
     maxAngularAcceleration=1000
     maxAngularVelocity=3
     maxLinearAcceleration=300
     maxLinearVelocity=700
     velocityScaleZ=0
     externalLinearDragV=<0 0.5 0>
    }
   }
   skydive_GoingUp {simSpaceSettings {
     maxAngularAcceleration=1000
     maxAngularVelocity=3
     maxLinearAcceleration=300
     maxLinearVelocity=700
     velocityScaleZ=0
     externalLinearDragV=<0 0.5 0>
    }}
   skydive_Parachuting {
    componentLinearAccScale=<0 0.1 0.3>
    componentLinearVelScale=<0 0.1 0.3>
    simSpaceSettings {
     maxAngularAcceleration=1000
     maxAngularVelocity=3
     maxLinearAcceleration=300
     maxLinearVelocity=700
     velocityScaleZ=0
     externalLinearDragV=<0 0.5 0>
    }
   }
   skydive_Surfing {simSpaceSettings {
     maxAngularAcceleration=1000
     maxAngularVelocity=3
     maxLinearAcceleration=300
     maxLinearVelocity=700
     velocityScaleZ=0
     externalLinearDragV=<0 0.5 0>
    }}
   swimming {simSpaceSettings {
     maxAngularAcceleration=1000
     maxAngularVelocity=3
     maxLinearAcceleration=300
     maxLinearVelocity=700
     velocityScaleZ=0
     externalLinearDragV=<0 0.5 0>
    }}
   tactical_Sprint {simSpaceSettings {
     maxAngularAcceleration=1000
     maxAngularVelocity=3
     maxLinearAcceleration=300
     maxLinearVelocity=700
     velocityScaleZ=0
     externalLinearDragV=<0 0.5 0>
    }}
   ziplining {simSpaceSettings {
     maxAngularAcceleration=1000
     maxAngularVelocity=3
     maxLinearAcceleration=300
     maxLinearVelocity=700
     velocityScaleZ=0
     externalLinearDragV=<0 0.5 0>
    }}
  }
 }
 Sim Space Settings {
  type=SimSpaceSettings
  default {
   dampingAlpha=1
   maxAngularAcceleration=10000
   maxAngularVelocity=10000
   maxLinearAcceleration=10000
   maxLinearVelocity=10000
   velocityScaleZ=1
   worldAlpha=0.3
   externalAngularVelocity=<0 1 1>
   externalLinearDragV=<0 0 0>
   externalLinearVelocity=<0 1 1>
  }
 }
 SprinitingNotCrouching {type=boolean}
 WalkinCrouching {type=boolean}
]
