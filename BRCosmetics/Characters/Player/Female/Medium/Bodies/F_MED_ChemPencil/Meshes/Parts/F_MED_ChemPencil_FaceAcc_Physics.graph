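# Physics asset for the ChemPencil female character's face accessory part, set up for a dynamic front hair strand.
# Three hair capsules chain from a small front-hair base on the head, with no twist and swing limits that widen toward the tip.
# Head, upper spine, right clavicle and right upper arm capsules give body volumes around the strand, joined by free angular constraints.
# The bun has only a tiny base body on the head.
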PhysicsAsset /BRCosmetics/Characters/Player/Female/Medium/Bodies/F_MED_ChemPencil/Meshes/Parts/F_MED_ChemPencil_FaceAcc_Physics
bodies {
 base_bun_cr [{shapeType=Capsule, length=0.505, radius=0.505, shapeName=base_bun_cr_capsule, rotation=<rot 0 -90 0>}]
 base_hair_front_cr [{shapeType=Capsule, length=0.505, radius=0.505, shapeName=base_hair_front_cr_capsule, rotation=<rot 0 -90 0>}]
 clavicle_r [{shapeType=Capsule, length=8.4581, radius=6.2128, shapeName=clavicle_r_capsule, center=<-8.1327 2.0677 0>, rotation=<rot -0.0002 -90 90>}]
 dyn_hair_front_01_cr [{
   shapeType=Capsule
   length=3.9908
   radius=1.224
   shapeName=dyn_hair_front_01_cr_capsule
   center=<3.6024 0.113 0.1188>
   rotation=<rot 0 -90.3006 -85.2483>
  }]
 dyn_hair_front_02_cr [{
   shapeType=Capsule
   length=3.5443
   radius=1.3011
   shapeName=dyn_hair_front_02_cr_capsule
   center=<3.3939 -0.2052 -0.2576>
   rotation=<rot 0 -89.1152 -87.7277>
  }]
 dyn_hair_front_03_cr [{
   shapeType=Capsule
   length=3.6671
   radius=1.4904
   shapeName=dyn_hair_front_03_cr_capsule
   center=<3.1501 -1.1019 -0.6812>
   rotation=<rot 0 -123.7251 -75.1037>
  }]
 head [{shapeType=Capsule, length=5.154, radius=5.6718, center=<-1.8747 3.5976 1.4033>, rotation=<rot 0 49.9997 -60>}]
 spine_05 [{shapeType=Capsule, length=11.5564, radius=11.4514, shapeName=spine_05_capsule, center=<-4.2793 -0.5837 0>, rotation=<rot 0 -90 0>}]
 upperarm_r [{shapeType=Capsule, length=34.4291, radius=5.544, shapeName=upperarm_r_capsule, center=<-22.1094 0 0>, rotation=<rot -0.0001 -89.9999 90>}]
}
constraints [
 {
  bone1Name=base_bun_cr
  bone2Name=head
  swing1LimitDegrees=45
  swing1Motion=Free
  swing2LimitDegrees=45
  swing2Motion=Free
  twistLimitDegrees=45
  twistMotion=Free
 }
 {
  bone1Name=base_hair_front_cr
  bone2Name=head
  swing1LimitDegrees=45
  swing1Motion=Free
  swing2LimitDegrees=45
  swing2Motion=Free
  twistLimitDegrees=45
  twistMotion=Free
 }
 {bone1Name=dyn_hair_front_01_cr, bone2Name=base_hair_front_cr, swing1LimitDegrees=5, swing2LimitDegrees=10, twistLimitDegrees=0}
 {bone1Name=dyn_hair_front_02_cr, bone2Name=dyn_hair_front_01_cr, swing1LimitDegrees=10, swing2LimitDegrees=17, twistLimitDegrees=0}
 {bone1Name=dyn_hair_front_03_cr, bone2Name=dyn_hair_front_02_cr, swing1LimitDegrees=15, swing2LimitDegrees=17, twistLimitDegrees=0}
 {
  bone1Name=upperarm_r
  bone2Name=clavicle_r
  swing1LimitDegrees=45
  swing1Motion=Free
  swing2LimitDegrees=45
  swing2Motion=Free
  twistLimitDegrees=45
  twistMotion=Free
 }
 {
  bone1Name=head
  bone2Name=spine_05
  swing1LimitDegrees=45
  swing1Motion=Free
  swing2LimitDegrees=45
  swing2Motion=Free
  twistLimitDegrees=45
  twistMotion=Free
 }
 {
  bone1Name=clavicle_r
  bone2Name=spine_05
  swing1LimitDegrees=45
  swing1Motion=Free
  swing2LimitDegrees=45
  swing2Motion=Free
  twistLimitDegrees=45
  twistMotion=Free
 }
]
